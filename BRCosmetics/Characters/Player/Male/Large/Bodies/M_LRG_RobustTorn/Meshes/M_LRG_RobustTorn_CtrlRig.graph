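# Deformation Control Rig for the large male RobustTorn player body, driving helper deform bones from the animated skeleton.
# Legs use constrained nulls feeding stretchy spline IK chains, with knee correctives weighted by bend angle.
# Arms copy onto deform chains solved with PBIK, plus twist distribution, shared deformation and pose-reader correctives, or plain constraints when disabled.
# Crouch, knee slide, pistol targeting and front-end variables gate extra pose fixes, mostly smoothed.
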ControlRigBlueprint /BRCosmetics/Characters/Player/Male/Large/Bodies/M_LRG_RobustTorn/Meshes/M_LRG_RobustTorn_CtrlRig
graphs [
 New Function_ContainedGraph {nodes [
   RigVMFunctionEntryNode()
   > RigVMFunction_Sequence = RigVMAggregateNode()
   &RigUnit_ParentConstraint(
    AdvancedSettings=(InterpolationType=Shortest,RotationOrderForFilter=XZY)
    Child=(Type=Bone,Name="dfrm_thigh_l")
    Filter=(TranslationFilter=(bX=true,bY=true,bZ=true),RotationFilter=(bX=True,bY=True,bZ=True),ScaleFilter=(bX=False,bY=False,bZ=False))
    Parents=((Item=(Type=Bone,Name="thigh_l"),Weight=1))
    Weight=1
    ExecutePin=@RigVMFunction_Sequence.A
   )
   > &RigUnit_ParentConstraint(
    AdvancedSettings=(InterpolationType=Shortest,RotationOrderForFilter=XZY)
    Child=(Type=Bone,Name="dfrm_calf_l")
    Filter=(TranslationFilter=(bX=true,bY=true,bZ=true),RotationFilter=(bX=True,bY=True,bZ=True),ScaleFilter=(bX=True,bY=True,bZ=True))
    Parents=((Item=(Type=Bone,Name="calf_l"),Weight=1))
    Weight=1
   )
   > &RigUnit_ParentConstraint(
    AdvancedSettings=(InterpolationType=Shortest,RotationOrderForFilter=XZY)
    Child=(Type=Bone,Name="dfrm_foot_l")
    Filter=(TranslationFilter=(bX=true,bY=true,bZ=true),RotationFilter=(bX=True,bY=True,bZ=True),ScaleFilter=(bX=True,bY=True,bZ=True))
    Parents=((Item=(Type=Bone,Name="foot_l"),Weight=1))
    Weight=1
   )
   > &RigUnit_ParentConstraint(
    AdvancedSettings=(InterpolationType=Shortest,RotationOrderForFilter=XZY)
    Child=(Type=Bone,Name="dfrm_thigh_r")
    Filter=(TranslationFilter=(bX=true,bY=true,bZ=true),RotationFilter=(bX=True,bY=True,bZ=True),ScaleFilter=(bX=False,bY=False,bZ=False))
    Parents=((Item=(Type=Bone,Name="thigh_r"),Weight=1))
    Weight=1
    bMaintainOffset=true
   )
   > &RigUnit_ParentConstraint(
    AdvancedSettings=(InterpolationType=Shortest,RotationOrderForFilter=XZY)
    Child=(Type=Bone,Name="dfrm_calf_r")
    Filter=(TranslationFilter=(bX=true,bY=true,bZ=true),RotationFilter=(bX=True,bY=True,bZ=True),ScaleFilter=(bX=True,bY=True,bZ=True))
    Parents=((Item=(Type=Bone,Name="calf_r"),Weight=1))
    Weight=1
   )
   > &RigUnit_ParentConstraint(
    AdvancedSettings=(InterpolationType=Shortest,RotationOrderForFilter=XZY)
    Child=(Type=Bone,Name="dfrm_foot_r")
    Filter=(TranslationFilter=(bX=true,bY=true,bZ=true),RotationFilter=(bX=True,bY=True,bZ=True),ScaleFilter=(bX=True,bY=True,bZ=True))
    Parents=((Item=(Type=Bone,Name="foot_r"),Weight=1))
    Weight=1
   )
   > &RigUnit_ParentConstraint(
    AdvancedSettings=(InterpolationType=Shortest,RotationOrderForFilter=XZY)
    Child=(Type=Bone,Name="dfrm_knee_out_base_l")
    Filter=(TranslationFilter=(bX=true,bY=true,bZ=true),RotationFilter=(bX=True,bY=True,bZ=True),ScaleFilter=(bX=True,bY=True,bZ=True))
    Parents=((Item=(Type=Bone,Name="deform_knee_out_base_l"),Weight=1))
    Weight=1
   )
   > &RigUnit_ParentConstraint(
    AdvancedSettings=(InterpolationType=Shortest,RotationOrderForFilter=XZY)
    Child=(Type=Bone,Name="dfrm_knee_out_l")
    Filter=(TranslationFilter=(bX=true,bY=true,bZ=true),RotationFilter=(bX=True,bY=True,bZ=True),ScaleFilter=(bX=True,bY=True,bZ=True))
    Parents=((Item=(Type=Bone,Name="deform_knee_out_l"),Weight=1))
    Weight=1
   )
   > &RigUnit_ParentConstraint(
    AdvancedSettings=(InterpolationType=Shortest,RotationOrderForFilter=XZY)
    Child=(Type=Bone,Name="dfrm_knee_in_base_l")
    Filter=(TranslationFilter=(bX=true,bY=true,bZ=true),RotationFilter=(bX=True,bY=True,bZ=True),ScaleFilter=(bX=True,bY=True,bZ=True))
    Parents=((Item=(Type=Bone,Name="deform_knee_in_base_l"),Weight=1))
    Weight=1
   )
   > &RigUnit_ParentConstraint(
    AdvancedSettings=(InterpolationType=Shortest,RotationOrderForFilter=XZY)
    Child=(Type=Bone,Name="dfrm_knee_in_l")
    Filter=(TranslationFilter=(bX=true,bY=true,bZ=true),RotationFilter=(bX=True,bY=True,bZ=True),ScaleFilter=(bX=True,bY=True,bZ=True))
    Parents=((Item=(Type=Bone,Name="deform_knee_in_l"),Weight=1))
    Weight=1
   )
   > &RigUnit_ParentConstraint(
    AdvancedSettings=(InterpolationType=Shortest,RotationOrderForFilter=XZY)
    Child=(Type=Bone,Name="dfrm_knee_out_base_r")
    Filter=(TranslationFilter=(bX=true,bY=true,bZ=true),RotationFilter=(bX=True,bY=True,bZ=True),ScaleFilter=(bX=True,bY=True,bZ=True))
    Parents=((Item=(Type=Bone,Name="deform_knee_out_base_r"),Weight=1))
    Weight=1
   )
   > &RigUnit_ParentConstraint(
    AdvancedSettings=(InterpolationType=Shortest,RotationOrderForFilter=XZY)
    Child=(Type=Bone,Name="dfrm_knee_out_r")
    Filter=(TranslationFilter=(bX=true,bY=true,bZ=true),RotationFilter=(bX=True,bY=True,bZ=True),ScaleFilter=(bX=True,bY=True,bZ=True))
    Parents=((Item=(Type=Bone,Name="deform_knee_out_r"),Weight=1))
    Weight=1
   )
   > &RigUnit_ParentConstraint(
    AdvancedSettings=(InterpolationType=Shortest,RotationOrderForFilter=XZY)
    Child=(Type=Bone,Name="dfrm_knee_in_base_r")
    Filter=(TranslationFilter=(bX=true,bY=true,bZ=true),RotationFilter=(bX=True,bY=True,bZ=True),ScaleFilter=(bX=True,bY=True,bZ=True))
    Parents=((Item=(Type=Bone,Name="deform_knee_in_base_r"),Weight=1))
    Weight=1
   )
   > &RigUnit_ParentConstraint(
    AdvancedSettings=(InterpolationType=Shortest,RotationOrderForFilter=XZY)
    Child=(Type=Bone,Name="dfrm_knee_in_r")
    Filter=(TranslationFilter=(bX=true,bY=true,bZ=true),RotationFilter=(bX=True,bY=True,bZ=True),ScaleFilter=(bX=True,bY=True,bZ=True))
    Parents=((Item=(Type=Bone,Name="deform_knee_in_r"),Weight=1))
    Weight=1
   )
   /Script/ControlRig.RigUnit_PositionConstraintLocalSpaceOffset(
    Child=(Type=Null,Name="dfrm_leg_spline_01_l_ctrl_null")
    Filter=(bX=True,bY=True,bZ=True)
    Parents=((Item=(Type=Bone,Name="dfrm_thigh_l"),Weight=0.2),(Item=(Type=Bone,Name="pelvis"),Weight=0.8))
    Weight=0.5
    bMaintainOffset=True
    ExecutePin=@RigVMFunction_Sequence.B
   ) {is=struct RigVMUnitNode}
   > &RigUnit_RotationConstraintLocalSpaceOffset(
    AdvancedSettings=(RotationOrderForFilter=XZY)
    Child=(Type=Null,Name="dfrm_leg_spline_01_l_ctrl_null")
    Filter=(bX=True,bY=True,bZ=True)
    Parents=((Item=(Type=Bone,Name="dfrm_thigh_l"),Weight=1))
    Weight=1
   )
   > &RigUnit_ParentConstraint(
    AdvancedSettings=(RotationOrderForFilter=XZY)
    Child=(Type=Null,Name="dfrm_leg_spline_03_l_ctrl_null")
    Parents=((Item=(Type=Bone,Name="dfrm_thigh_l"),Weight=1),(Item=(Type=Bone,Name="dfrm_calf_l"),Weight=0))
    Weight=1
   )
   > &RigUnit_ParentConstraint(
    AdvancedSettings=(RotationOrderForFilter=XZY)
    Child=(Type=Null,Name="dfrm_leg_spline_05_l_ctrl_null")
    Parents=((Item=(Type=Bone,Name="dfrm_thigh_l"),Weight=0.5),(Item=(Type=Bone,Name="dfrm_calf_l"),Weight=0.5))
    Weight=1
   )
   > &RigUnit_ParentConstraint(
    AdvancedSettings=(RotationOrderForFilter=XZY)
    Child=(Type=Null,Name="dfrm_leg_spline_07_l_ctrl_null")
    Parents=((Item=(Type=Bone,Name="dfrm_calf_l"),Weight=1),(Item=(Type=Bone,Name="dfrm_foot_l"),Weight=0))
    Weight=1
   )
   > DeltaRotation Z = RigVMFunctionReferenceNode(Child=calf_l, Parent=thigh_l, SourceMax=120, SourceMin=100)
   > &RigUnit_ModifyTransforms(
    ItemToModify="((Item=(Type=Bone,Name=\"deform_knee_in_l\"),Transform=(Rotation=(X=0,Y=0,Z=0,W=1),Translation=(X=1.0749,Y=-38.8539,Z=0),Scale3D=(X=1,Y=1,Z=1))),(Item=(Type=Bone,Name=\"deform_knee_out_l\"),Transform=(Rotation=(X=0,Y=0,Z=0,W=1),Translation=(X=1.8018,Y=18.4981, ...(truncated)"
    WeightMinimum=0
    Weight=@"DeltaRotation Z".OuputWeight
   )
   ItemArray_1 = RigVMDispatchNode(Value="((Type=Bone,Name=\"dfrm_leg_spline_01_l\"),(Type=Bone,Name=\"dfrm_leg_spline_02_l\"),(Type=Bone,Name=\"dfrm_leg_spline_03_l\"),(Type=Bone,Name=\"dfrm_leg_spline_04_l\"),(Type=Bone,Name=\"dfrm_leg_spline_05_l\"),(Type=Bone,Name=\"dfrm_leg_spline_06_l\"),(Type=Bone,Name ...(truncated)") {template="DISPATCH_RigVMDispatch_Constant(io Value)"}
   > SplineIK_2 = RigVMFunctionReferenceNode(
    Primary Axis=(X=1,Y=0,Z=0)
    Samples Per Segment=6
    Secondary Spline Direction=(X=0,Y=0,Z=0)
    Spline Mode=BSpline
    Stretch=true
    Up Axis=(X=0,Y=0,Z=0)
    debug=false
    Bones=@ItemArray_1.Value
    Controls=RigVMVariableNode() {variable=L_legSpline_ctrls}.Value
   )
   > &RigUnit_ModifyTransforms(
    ItemToModify="((Item=(Type=Bone,Name=\"dfrm_leg_spline_06_l\"),Transform=(Rotation=(X=0,Y=0,Z=0,W=-1),Translation=(X=0,Y=0,Z=0),Scale3D=(X=1,Y=1,Z=1.4196))),(Item=(Type=Bone,Name=\"dfrm_leg_spline_07_l\"),Transform=(Rotation=(X=0,Y=0,Z=0,W=-1),Translation=(X=0,Y=0,Z=0),Scal ...(truncated)"
    WeightMinimum=0
    Weight=@"DeltaRotation Z".OuputWeight
   )
   &RigUnit_PositionConstraintLocalSpaceOffset(
    Child=(Type=Null,Name="dfrm_leg_spline_01_r_ctrl_null")
    Filter=(bX=True,bY=True,bZ=True)
    Parents=((Item=(Type=Bone,Name="dfrm_thigh_r"),Weight=0.2),(Item=(Type=Bone,Name="pelvis"),Weight=0.8))
    Weight=1
    bMaintainOffset=True
    ExecutePin=@RigVMFunction_Sequence.C
   )
   > &RigUnit_RotationConstraintLocalSpaceOffset(
    AdvancedSettings=(RotationOrderForFilter=XZY)
    Child=(Type=Null,Name="dfrm_leg_spline_01_r_ctrl_null")
    Filter=(bX=True,bY=True,bZ=True)
    Parents=((Item=(Type=Bone,Name="dfrm_thigh_r"),Weight=1))
    Weight=1
   )
   > &RigUnit_ParentConstraint(
    AdvancedSettings=(RotationOrderForFilter=XZY)
    Child=(Type=Null,Name="dfrm_leg_spline_03_r_ctrl_null")
    Parents=((Item=(Type=Bone,Name="dfrm_thigh_r"),Weight=1),(Item=(Type=Bone,Name="dfrm_calf_r"),Weight=0))
    Weight=1
   )
   > &RigUnit_ParentConstraint(
    AdvancedSettings=(RotationOrderForFilter=XZY)
    Child=(Type=Null,Name="dfrm_leg_spline_05_r_ctrl_null")
    Parents=((Item=(Type=Bone,Name="dfrm_thigh_r"),Weight=0.5),(Item=(Type=Bone,Name="dfrm_calf_r"),Weight=0.5))
    Weight=1
   )
   > &RigUnit_ParentConstraint(
    AdvancedSettings=(RotationOrderForFilter=XZY)
    Child=(Type=Null,Name="dfrm_leg_spline_07_r_ctrl_null")
    Parents=((Item=(Type=Bone,Name="dfrm_calf_r"),Weight=1),(Item=(Type=Bone,Name="dfrm_foot_r"),Weight=0))
    Weight=1
   )
   > DeltaRotation Z_1 = RigVMFunctionReferenceNode(Child=calf_r, Parent=thigh_r, SourceMax=120, SourceMin=110)
   > &RigUnit_ModifyTransforms(
    ItemToModify="((Item=(Type=Bone,Name=\"deform_knee_in_r\"),Transform=(Rotation=(X=0,Y=0,Z=-0.3856,W=-0.9227),Translation=(X=17.5343,Y=16.4596,Z=-0.0003),Scale3D=(X=1,Y=3.3024,Z=1))),(Item=(Type=Bone,Name=\"deform_knee_out_r\"),Transform=(Rotation=(X=0,Y=0,Z=0,W=-1),Translat ...(truncated)"
    WeightMinimum=0
    Weight=@"DeltaRotation Z_1".OuputWeight
   )
   RigVMFunction_MathBoolToFloat = &RigVMFunction_MathBoolToFloat(Value=RigVMVariableNode() {variable=Crouching}.Value)
   AlphaInterp = &RigVMFunction_AlphaInterp(InterpSpeedDecreasing=10, InterpSpeedIncreasing=10, Scale=1, bClampResult=true, Value=@RigVMFunction_MathBoolToFloat.Result)
   > &RigUnit_ModifyTransforms(
    ItemToModify="((Item=(Type=Bone,Name=\"pelvis_extra\"),Transform=(Rotation=(X=0,Y=0,Z=0,W=1),Translation=(X=1.6398,Y=0,Z=0),Scale3D=(X=1,Y=1,Z=1))),(Item=(Type=Null,Name=\"dfrm_leg_spline_03_r_ctrl_null\"),Transform=(Rotation=(X=0,Y=0,Z=0,W=1),Translation=(X=0,Y=0,Z=0),Scal ...(truncated)"
    WeightMinimum=0
    Weight=@AlphaInterp.Result
   )
   ItemArray_3 = RigVMDispatchNode(Value="((Type=Bone,Name=\"dfrm_leg_spline_01_r\"),(Type=Bone,Name=\"dfrm_leg_spline_02_r\"),(Type=Bone,Name=\"dfrm_leg_spline_03_r\"),(Type=Bone,Name=\"dfrm_leg_spline_04_r\"),(Type=Bone,Name=\"dfrm_leg_spline_05_r\"),(Type=Bone,Name=\"dfrm_leg_spline_06_r\"),(Type=Bone,Name ...(truncated)") {template="DISPATCH_RigVMDispatch_Constant(io Value)"}
   > SplineIK_3 = RigVMFunctionReferenceNode(
    Primary Axis=(X=-1,Y=0,Z=0)
    Samples Per Segment=6
    Secondary Spline Direction=(X=0,Y=0,Z=0)
    Spline Mode=BSpline
    Stretch=true
    Up Axis=(X=0,Y=0,Z=0)
    debug=false
    Bones=@ItemArray_3.Value
    Controls=RigVMVariableNode() {variable=R_legSpline_ctrls}.Value
   )
   > &RigUnit_ModifyTransforms(
    ItemToModify="((Item=(Type=Bone,Name=\"dfrm_leg_spline_06_r\"),Transform=(Rotation=(X=0,Y=0,Z=0,W=-1),Translation=(X=0,Y=0,Z=0),Scale3D=(X=1,Y=0.87,Z=1))),(Item=(Type=Bone,Name=\"dfrm_leg_spline_07_r\"),Transform=(Rotation=(X=0,Y=0,Z=0,W=1),Translation=(X=0,Y=0,Z=0),Scale3D ...(truncated)"
    WeightMinimum=0
    Weight=@"DeltaRotation Z_1".OuputWeight
   )
   RigVMFunctionReturnNode(ExecuteContext=@RigVMFunction_Sequence.E)
   n1 = &RigVMFunction_AlphaInterp(
    InterpSpeedDecreasing=10
    InterpSpeedIncreasing=10
    Scale=1
    bClampResult=true
    Value=&RigVMFunction_MathBoolToFloat(Value=RigVMVariableNode() {variable=kneeSliding}.Value).Result
   )
   &RigUnit_ModifyTransforms(
    ItemToModify="((Item=(Type=Bone,Name=\"dfrm_leg_spline_07_r\"),Transform=(Rotation=(X=0,Y=0.0855,Z=0,W=0.9963),Translation=(X=0,Y=0,Z=0),Scale3D=(X=1,Y=1,Z=1))),(Item=(Type=Bone,Name=\"dfrm_leg_spline_01_r\"),Transform=(Rotation=(X=0,Y=0.0116,Z=-0.1206,W=0.9926),Translation ...(truncated)"
    WeightMinimum=0
    Weight=@n1.Result
    ExecutePin=@RigVMFunction_Sequence.D
   )
  ]}
 New Function_ContainedGraph {nodes [
   n1 = RigVMFunctionEntryNode()
   RigVMFunction_MathQuaternionToEuler = &RigVMFunction_MathQuaternionToEuler(
    RotationOrder=XYZ
    Value=&RigUnit_GetRelativeTransformForItem(
     Child=(Type=Bone,Name="hand_l")
     Parent=(Type=Bone,Name="lowerarm_l")
     bChildInitial=False
     bParentInitial=False
     Child.Name=@n1.Child
     Parent.Name=@n1.Parent
    ).RelativeTransform.Rotation
   )
   Remap = &RigVMFunction_MathDoubleRemap(
    TargetMaximum=1
    TargetMinimum=0
    Value=@RigVMFunction_MathQuaternionToEuler.Result.Z
    SourceMaximum=@n1.SourceMax
    SourceMinimum=@n1.SourceMin
   )
   > RigVMFunctionReturnNode(OuputWeight=@Remap.Result)
  ]}
 New Function_ContainedGraph {nodes [
   n1 = RigVMFunctionEntryNode()
   RigVMFunction_ControlFlowBranch = &RigVMFunction_ControlFlowBranch(Condition=@n1.Controls, ExecuteContext=@n1.ExecuteContext)
   RigVMFunctionReturnNode(ExecuteContext=@RigVMFunction_ControlFlowBranch.Completed)
   RigUnit_SetTransform = &RigUnit_SetTransform(
    Item=(Type=Bone,Name="pelvis")
    Weight=1
    bInitial=False
    Value=&RigUnit_GetTransform(Item=(Type=Control,Name="pelvis_ctrl"), bInitial=False).Transform
    ExecutePin=@RigVMFunction_ControlFlowBranch.True
   )
   RigUnit_GetTransform = &RigUnit_GetTransform(Item=(Type=Control,Name="thigh_r_ctrl"), bInitial=False)
   > &RigUnit_SetTransform(Item=(Type=Bone,Name="thigh_r"), Weight=1, bInitial=False, Value=@RigUnit_GetTransform.Transform)
   > &RigUnit_SetTransform(
    Item=(Type=Bone,Name="thigh_l")
    Weight=1
    bInitial=False
    Value=&RigUnit_GetTransform(Item=(Type=Control,Name="thigh_l_ctrl"), bInitial=False).Transform
   )
   > &RigUnit_SetTransform(
    Item=(Type=Bone,Name="calf_l")
    Weight=1
    bInitial=False
    Value=&RigUnit_GetTransform(Item=(Type=Control,Name="calf_l_ctrl"), bInitial=False).Transform
   )
   > &RigUnit_SetTransform(
    Item=(Type=Bone,Name="calf_r")
    Weight=1
    bInitial=False
    Value=&RigUnit_GetTransform(Item=(Type=Control,Name="calf_r_ctrl"), bInitial=False).Transform
   )
   > &RigUnit_SetTransform(
    Item=(Type=Bone,Name="foot_l")
    Weight=1
    bInitial=False
    Value=&RigUnit_GetTransform(Item=(Type=Control,Name="foot_l_ctrl"), bInitial=False).Transform
   )
   > &RigUnit_SetTransform(
    Item=(Type=Bone,Name="foot_r")
    Weight=1
    bInitial=False
    Value=&RigUnit_GetTransform(Item=(Type=Control,Name="foot_r_ctrl"), bInitial=False).Transform
   )
  ]}
 New Function_ContainedGraph {nodes [
   n1 = RigVMFunctionEntryNode()
   RigVMFunction_ControlFlowBranch = &RigVMFunction_ControlFlowBranch(Condition=@n1.Disable Arms, ExecuteContext=@n1.ExecuteContext)
   &RigUnit_ModifyTransforms(
    ItemToModify="((Item=(Type=Bone,Name=\"deform_upperarm_up_l\"),Transform=(Rotation=(X=0,Y=0,Z=0,W=1),Translation=(X=0,Y=0,Z=0),Scale3D=(X=1,Y=1,Z=1))),(Item=(Type=Bone,Name=\"deform_upperarm_twist_01_l\"),Transform=(Rotation=(X=0,Y=0,Z=0,W=1),Translation=(X=0,Y=0,Z=-6),Scal ...(truncated)"
    WeightMinimum=0
    Weight=RigVMDispatchNode(BoneName=root, Default=0, Name=deformRig_l_upperarm_up_alpha) {template="DISPATCH_RigDispatch_GetAnimAttribute(in Name,in BoneName,in Default,out Value,out Found)"}.Value
    ExecutePin=@RigVMFunction_ControlFlowBranch.Completed
   )
   n2 = &RigVMFunction_AlphaInterp(
    InterpSpeedDecreasing=10
    InterpSpeedIncreasing=10
    Scale=1
    bClampResult=true
    Value=&RigVMFunction_MathBoolToFloat(Value=RigVMVariableNode() {variable=Crouching}.Value).Result
   )
   Remap = &RigVMFunction_MathDoubleRemap(SourceMaximum=1, SourceMinimum=0, TargetMaximum=0, TargetMinimum=1, bClamp=False, Value=@n2.Result)
   > &RigUnit_ModifyTransforms(
    ItemToModify="((Item=(Type=Bone,Name=\"deform_upperarm_up_r\"),Transform=(Rotation=(X=0,Y=0,Z=0,W=1),Translation=(X=0,Y=0,Z=0),Scale3D=(X=1,Y=1,Z=1))),(Item=(Type=Bone,Name=\"deform_upperarm_twist_01_r\"),Transform=(Rotation=(X=0,Y=0,Z=0,W=-1),Translation=(X=0,Y=0,Z=6.3564) ...(truncated)"
    WeightMinimum=0
    Weight=&RigVMFunction_MathDoubleMul(
     A=RigVMDispatchNode(BoneName=root, Default=0, Name=deformRig_r_upperarm_up_alpha) {template="DISPATCH_RigDispatch_GetAnimAttribute(in Name,in BoneName,in Default,out Value,out Found)"}.Value
     B=@Remap.Result
    ).Result
   )
   > n3 = &RigUnit_SphericalPoseReader(
    ActiveRegionScaleFactors=(PositiveWidth=2.5,NegativeWidth=3,PositiveHeight=3,NegativeHeight=3)
    ActiveRegionSize=0.1
    Debug=(bDrawDebug=false,DebugScale=25,DebugSegments=20,DebugThickness=0.25)
    DriverAxis=(X=-1,Y=0,Z=0)
    DriverItem=(Type=Bone,Name="upperarm_r")
    FalloffRegionScaleFactors=(PositiveWidth=1,NegativeWidth=1,PositiveHeight=1,NegativeHeight=1)
    FalloffSize=0.65
    FlipHeightScaling=true
    OptionalParentItem=(Type=Bone)
    RotationOffset=(X=180,Y=0,Z=180)
   )
   RigVMFunction_MathBoolToFloat = &RigVMFunction_MathBoolToFloat(Value=RigVMVariableNode() {variable=Crouching}.Value)
   AlphaInterp = &RigVMFunction_AlphaInterp(InterpSpeedDecreasing=10, InterpSpeedIncreasing=10, Scale=1, bClampResult=true, Value=@RigVMFunction_MathBoolToFloat.Result)
   Multiply = &RigVMFunction_MathDoubleMul(B=@AlphaInterp.Result, A=@n3.OutputParam)
   RigVMFunction_AnimEvalRichCurve = &RigVMFunction_AnimEvalRichCurve(
    Curve=(EditorCurveData=(Keys=((),(InterpMode=RCIM_Cubic,Time=0.2763,Value=0.59,ArriveTangent=1,LeaveTangent=1),(Time=1,Value=1))))
    SourceMinimum=0
    TargetMinimum=0
    Value=@Multiply.Result
   )
   > &RigUnit_ModifyTransforms(
    ItemToModify=((Item=(Type=Bone,Name="deform_upperarm_bk_r"),Transform=(Rotation=(X=-0.338,Y=0.6826,Z=0.4052,W=0.5055),Translation=(X=9.8082,Y=16.6868,Z=9.4),Scale3D=(X=1,Y=1,Z=1))))
    WeightMinimum=0
    Weight=@RigVMFunction_AnimEvalRichCurve.Result
   )
   > RigVMFunctionReturnNode()
   RigVMFunction_Sequence = &RigVMFunction_Sequence(ExecuteContext=@RigVMFunction_ControlFlowBranch.False)
   ArmIKCorrections = RigVMFunctionReferenceNode(DisableArmsModifications=False, ExecuteContext=@RigVMFunction_Sequence.A)
   > ZebraScramble_SharedDeformation_1 = RigVMFunctionReferenceNode(
    disable_fingers=0
    value_struct="(upperarm_up_18_#1=(Rotation=(X=0,Y=0,Z=0,W=1),Translation=(X=0,Y=0,Z=0),Scale3D=(X=1,Y=1,Z=1)),upperarm_fwd_fr_19_#2=(Rotation=(X=0,Y=0,Z=0,W=1),Translation=(X=0,Y=10,Z=10),Scale3D=(X=1,Y=1,Z=1)) ...(truncated)"
   )
   DeformTwist = RigVMFunctionReferenceNode(ExecuteContext=@RigVMFunction_Sequence.B)
   &RigUnit_ParentConstraint(
    AdvancedSettings=(InterpolationType=Average,RotationOrderForFilter=XZY)
    Child=(Type=Bone,Name="deform_upperarm_l")
    Parents=((Item=(Type=Bone,Name="upperarm_l"),Weight=1))
    Weight=1
    ExecutePin=@RigVMFunction_ControlFlowBranch.True
   )
   > &RigUnit_ParentConstraint(
    AdvancedSettings=(InterpolationType=Average,RotationOrderForFilter=XZY)
    Child=(Type=Bone,Name="deform_upperarm_twist_01_l")
    Parents=((Item=(Type=Bone,Name="upperarm_twist_01_l"),Weight=1))
    Weight=1
   )
   > &RigUnit_ParentConstraint(
    AdvancedSettings=(InterpolationType=Average,RotationOrderForFilter=XZY)
    Child=(Type=Bone,Name="deform_upperarm_twist_02_l")
    Parents=((Item=(Type=Bone,Name="upperarm_twist_02_l"),Weight=1))
    Weight=1
   )
   > &RigUnit_ParentConstraint(
    AdvancedSettings=(InterpolationType=Average,RotationOrderForFilter=XZY)
    Child=(Type=Bone,Name="deform_lowerarm_l")
    Parents=((Item=(Type=Bone,Name="lowerarm_l"),Weight=1))
    Weight=1
   )
   > &RigUnit_ParentConstraint(
    AdvancedSettings=(InterpolationType=Average,RotationOrderForFilter=XZY)
    Child=(Type=Bone,Name="deform_lowerarm_twist_01_l")
    Parents=((Item=(Type=Bone,Name="lowerarm_twist_01_l"),Weight=1))
    Weight=1
   )
   > &RigUnit_ParentConstraint(
    AdvancedSettings=(InterpolationType=Average,RotationOrderForFilter=XZY)
    Child=(Type=Bone,Name="deform_lowerarm_twist_02_l")
    Parents=((Item=(Type=Bone,Name="lowerarm_twist_02_l"),Weight=1))
    Weight=1
   )
   > &RigUnit_ParentConstraint(
    AdvancedSettings=(InterpolationType=Average,RotationOrderForFilter=XZY)
    Child=(Type=Bone,Name="deform_hand_l")
    Parents=((Item=(Type=Bone,Name="hand_l"),Weight=1))
    Weight=1
   )
   > &RigUnit_ParentConstraint(
    AdvancedSettings=(InterpolationType=Average,RotationOrderForFilter=XZY)
    Child=(Type=Bone,Name="deform_upperarm_r")
    Parents=((Item=(Type=Bone,Name="upperarm_r"),Weight=1))
    Weight=1
   )
   > &RigUnit_ParentConstraint(
    AdvancedSettings=(InterpolationType=Average,RotationOrderForFilter=XZY)
    Child=(Type=Bone,Name="deform_upperarm_twist_01_r")
    Parents=((Item=(Type=Bone,Name="upperarm_twist_01_r"),Weight=1))
    Weight=1
   )
   > &RigUnit_ParentConstraint(
    AdvancedSettings=(InterpolationType=Average,RotationOrderForFilter=XZY)
    Child=(Type=Bone,Name="deform_upperarm_twist_02_r")
    Parents=((Item=(Type=Bone,Name="upperarm_twist_02_r"),Weight=1))
    Weight=1
   )
   > &RigUnit_ParentConstraint(
    AdvancedSettings=(InterpolationType=Average,RotationOrderForFilter=XZY)
    Child=(Type=Bone,Name="deform_lowerarm_r")
    Parents=((Item=(Type=Bone,Name="lowerarm_r"),Weight=1))
    Weight=1
   )
   > &RigUnit_ParentConstraint(
    AdvancedSettings=(InterpolationType=Average,RotationOrderForFilter=XZY)
    Child=(Type=Bone,Name="deform_lowerarm_twist_01_r")
    Parents=((Item=(Type=Bone,Name="lowerarm_twist_01_r"),Weight=1))
    Weight=1
   )
   > &RigUnit_ParentConstraint(
    AdvancedSettings=(InterpolationType=Average,RotationOrderForFilter=XZY)
    Child=(Type=Bone,Name="lowerarm_twist_02_r")
    Parents=((Item=(Type=Bone,Name="lowerarm_twist_02_r"),Weight=1))
    Weight=1
   )
   > &RigUnit_ParentConstraint(
    AdvancedSettings=(InterpolationType=Average,RotationOrderForFilter=XZY)
    Child=(Type=Bone,Name="deform_hand_r")
    Parents=((Item=(Type=Bone,Name="hand_r"),Weight=1))
    Weight=1
   )
  ]}
 New Function_ContainedGraph {nodes [
   n1 = RigVMFunctionEntryNode()
   > RigVMFunction_Sequence = RigVMAggregateNode()
   RigVMFunction_Sequence_1 = RigVMAggregateNode(ExecuteContext=@RigVMFunction_Sequence.A)
   n2 = RigVMDispatchNode(Array=((Type=Bone,Name="upperarm_l"),(Type=Bone,Name="lowerarm_l"),(Type=Bone,Name="hand_l")), ExecuteContext=@RigVMFunction_Sequence_1.A) {template="DISPATCH_RigVMDispatch_ArrayIterator(in Array,out Element,out Index,out Count,out Ratio)"}
   > &RigUnit_SetTransform(
    Weight=1
    bInitial=False
    Item=RigVMDispatchNode(Array=((Type=Bone,Name="deform_upperarm_l"),(Type=Bone,Name="deform_lowerarm_l"),(Type=Bone,Name="deform_hand_l")), Index=@n2.Index) {template="DISPATCH_RigVMDispatch_ArrayGetAtIndex(in Array,in Index,out Element)"}.Element
    Value=&RigUnit_GetTransform(bInitial=False, Item=@n2.Element).Transform
   )
   n3 = /Script/PBIK.RigUnit_PBIK(
    BoneSettings=((Bone="deform_hand_l",RotationStiffness=0,PositionStiffness=0,X=Free,MinX=0,MaxX=0,Y=Free,MinY=0,MaxY=0,Z=Free,MinZ=0,MaxZ=0,bUsePreferredAngles=true,PreferredAngles=(X=1,Y=0,Z=0)))
    Debug=(DrawScale=1,bDrawDebug=False)
    Effectors=((Bone="hand_l",Transform=(Rotation=(X=0,Y=0,Z=0,W=1),Translation=(X=0,Y=0,Z=0),Scale3D=(X=1,Y=1,Z=1)),PositionAlpha=1,RotationAlpha=1,StrengthAlpha=1,PullChainAlpha=1,PinRotation=1,ChainDepth=0))
    ExcludedBones=()
    Root=deform_upperarm_l
    Settings="(Iterations=20,SubIterations=0,MassMultiplier=1,bAllowStretch=false,RootBehavior=PrePull,PrePullRootSettings=(RotationAlpha=0,RotationAlphaX=1,RotationAlphaY=1,RotationAlphaZ=1,PositionAlpha=1,PositionAlphaX=1,PositionAlphaY=1,PositionAlphaZ=1),GlobalPullC ...(truncated)"
    ExecutePin=@n2.Completed
   ) {is=struct RigVMUnitNode}
   RigVMFunction_ControlFlowBranch = &RigVMFunction_ControlFlowBranch(ExecuteContext=@n3.ExecutePin, Condition=@n1.DisableArmsModifications)
   n4 = &RigUnit_SphericalPoseReader(
    ActiveRegionScaleFactors=(PositiveWidth=2,NegativeWidth=0.5,PositiveHeight=1,NegativeHeight=1)
    ActiveRegionSize=0.1
    Debug=(bDrawDebug=false,bDraw2D=False,bDrawLocalAxes=False,DebugScale=25,DebugSegments=20,DebugThickness=0.25)
    DriverAxis=(X=1,Y=0,Z=0)
    DriverItem=(Type=Bone,Name="upperarm_l")
    FalloffRegionScaleFactors=(PositiveWidth=1,NegativeWidth=1,PositiveHeight=2,NegativeHeight=2)
    FalloffSize=0.2
    FlipHeightScaling=False
    FlipWidthScaling=False
    OptionalParentItem=(Type=Bone,Name="spine_03")
    RotationOffset=(X=0,Y=-145,Z=0)
    ExecutePin=@RigVMFunction_ControlFlowBranch.False
   )
   > ApplyFBIKCahinCorrections = RigVMFunctionReferenceNode(End=(Type=Bone,Name="deform_hand_l"), QuatAmount=-0.4, Start=(Type=Bone,Name="deform_upperarm_l"), Weight=@n4.OutputParam)
   > Get Normalized Distance Between = RigVMFunctionReferenceNode(Max Distance=40, Min Distance=10, Source=(Type=Bone,Name="hand_l"), Target=(Type=Bone,Name="pelvis"))
   > ApplyFBIKCahinCorrections_1 = RigVMFunctionReferenceNode(
    End=(Type=Bone,Name="deform_hand_l")
    QuatAmount=-0.8
    Start=(Type=Bone,Name="deform_upperarm_l")
    Weight=@"Get Normalized Distance Between".Result
   )
   n5 = RigVMDispatchNode(Array=((Type=Bone,Name="upperarm_r"),(Type=Bone,Name="lowerarm_r"),(Type=Bone,Name="hand_r")), ExecuteContext=@RigVMFunction_Sequence_1.B) {template="DISPATCH_RigVMDispatch_ArrayIterator(in Array,out Element,out Index,out Count,out Ratio)"}
   > &RigUnit_SetTransform(
    Weight=1
    bInitial=False
    Item=RigVMDispatchNode(Array=((Type=Bone,Name="deform_upperarm_r"),(Type=Bone,Name="deform_lowerarm_r"),(Type=Bone,Name="deform_hand_r")), Index=@n5.Index) {template="DISPATCH_RigVMDispatch_ArrayGetAtIndex(in Array,in Index,out Element)"}.Element
    Value=&RigUnit_GetTransform(bInitial=False, Item=@n5.Element).Transform
   )
   n6 = &RigUnit_PBIK(
    BoneSettings="((Bone=\"deform_hand_r\",RotationStiffness=0,PositionStiffness=0,X=Free,MinX=0,MaxX=0,Y=Free,MinY=0,MaxY=0,Z=Free,MinZ=0,MaxZ=0,bUsePreferredAngles=false,PreferredAngles=(X=1,Y=0,Z=0)),(Bone=\"deform_upperarm_l\",RotationStiffness=0,PositionStiffness=0,X=Free, ...(truncated)"
    Debug=(DrawScale=1,bDrawDebug=False)
    Effectors="((Bone=\"hand_r\",Transform=(Rotation=(X=0,Y=0,Z=0,W=1),Translation=(X=0,Y=0,Z=0),Scale3D=(X=1,Y=1,Z=1)),PositionAlpha=1,RotationAlpha=1,StrengthAlpha=1,PullChainAlpha=1,PinRotation=1,ChainDepth=0),(Bone=\"upperarm_r\",Transform=(Rotation=(X=0,Y=0,Z=0,W=1),Tra ...(truncated)"
    ExcludedBones=()
    Root=deform_upperarm_r
    Settings="(Iterations=20,SubIterations=0,MassMultiplier=1,bAllowStretch=false,RootBehavior=PrePull,PrePullRootSettings=(RotationAlpha=0,RotationAlphaX=1,RotationAlphaY=1,RotationAlphaZ=1,PositionAlpha=1,PositionAlphaX=1,PositionAlphaY=1,PositionAlphaZ=1),GlobalPullC ...(truncated)"
    ExecutePin=@n5.Completed
   )
   n7 = &RigVMFunction_ControlFlowBranch(ExecuteContext=@n6.ExecutePin, Condition=@n1.DisableArmsModifications)
   n8 = &RigUnit_SphericalPoseReader(
    ActiveRegionScaleFactors=(PositiveWidth=2,NegativeWidth=0.5,PositiveHeight=1,NegativeHeight=1)
    ActiveRegionSize=0.1
    Debug=(bDrawDebug=false,bDraw2D=False,bDrawLocalAxes=False,DebugScale=25,DebugSegments=20,DebugThickness=0.25)
    DriverAxis=(X=-1,Y=0,Z=0)
    DriverItem=(Type=Bone,Name="upperarm_r")
    FalloffRegionScaleFactors=(PositiveWidth=1,NegativeWidth=1,PositiveHeight=2,NegativeHeight=2)
    FalloffSize=0.2
    FlipHeightScaling=False
    FlipWidthScaling=False
    OptionalParentItem=(Type=Bone,Name="spine_03")
    RotationOffset=(X=0,Y=35,Z=0)
    ExecutePin=@n7.False
   )
   > ApplyFBIKCahinCorrections_2 = RigVMFunctionReferenceNode(End=(Type=Bone,Name="deform_hand_r"), QuatAmount=0.4, Start=(Type=Bone,Name="deform_upperarm_r"), Weight=@n8.OutputParam)
   > Get Normalized Distance Between_1 = RigVMFunctionReferenceNode(Max Distance=35, Min Distance=1, Source=(Type=Bone,Name="hand_r"), Target=(Type=Bone,Name="spine_04"))
   > ApplyFBIKCahinCorrections_3 = RigVMFunctionReferenceNode(
    End=(Type=Bone,Name="deform_hand_r")
    QuatAmount=0.4
    Start=(Type=Bone,Name="deform_upperarm_r")
    Weight=@"Get Normalized Distance Between_1".Result
   )
   > n9 = &RigUnit_SphericalPoseReader(
    ActiveRegionScaleFactors=(PositiveWidth=3,NegativeWidth=3,PositiveHeight=3,NegativeHeight=3)
    ActiveRegionSize=0.1
    Debug=(bDrawDebug=false,DebugScale=25,DebugSegments=20,DebugThickness=0.25)
    DriverAxis=(X=-1,Y=0,Z=0)
    DriverItem=(Type=Bone,Name="lowerarm_r")
    FalloffRegionScaleFactors=(PositiveWidth=2.1,NegativeWidth=2,PositiveHeight=2,NegativeHeight=2)
    FalloffSize=0.39
    OptionalParentItem=(Type=Bone,Name="spine_05")
    RotationOffset=(X=50,Y=30,Z=90)
   )
   > &RigUnit_ModifyTransforms(
    ItemToModify="((Item=(Type=Bone,Name=\"deform_upperarm_twist_01_r\"),Transform=(Rotation=(X=0.1446,Y=-0.0044,Z=-0.0119,W=-0.9894),Translation=(X=-0.0042,Y=0.1443,Z=-0.0787),Scale3D=(X=0.9,Y=0.9,Z=0.9))),(Item=(Type=Bone,Name=\"deform_upperarm_twist_02_r\"),Transform=(Rotati ...(truncated)"
    WeightMinimum=0
    Weight=&RigVMFunction_MathDoubleMul(A=@n9.OutputParam, B=RigVMVariableNode() {variable=PistolTargeting}.Value).Result
   )
   > n10 = &RigUnit_SphericalPoseReader(
    ActiveRegionScaleFactors=(PositiveWidth=1,NegativeWidth=1,PositiveHeight=1,NegativeHeight=1)
    ActiveRegionSize=0.2
    Debug=(bDrawDebug=true,bDraw2D=False,bDrawLocalAxes=False,DebugScale=25,DebugSegments=20,DebugThickness=0.25)
    DriverAxis=(X=1,Y=0,Z=0)
    DriverItem=(Type=Bone,Name="upperarm_l")
    FalloffRegionScaleFactors=(PositiveWidth=1,NegativeWidth=1,PositiveHeight=1,NegativeHeight=1)
    FalloffSize=0.45
    FlipHeightScaling=False
    FlipWidthScaling=false
    OptionalParentItem=(Type=Bone,Name="spine_05")
    RotationOffset=(X=150,Y=40,Z=0)
   )
   AlphaInterp = &RigVMFunction_AlphaInterp(InterpSpeedDecreasing=5, InterpSpeedIncreasing=5, Scale=1, bClampResult=true, Value=RigVMVariableNode() {variable=FrontEndValue}.Value)
   n11 = &RigVMFunction_MathDoubleMul(B=@AlphaInterp.Result, A=@n10.OutputParam)
   > &RigUnit_ModifyTransforms(
    ItemToModify="((Item=(Type=Bone,Name=\"deform_upperarm_up_l\"),Transform=(Rotation=(X=0,Y=0,Z=0,W=1),Translation=(X=-17.9735,Y=-5.266,Z=7.4589),Scale3D=(X=1.5,Y=2,Z=1))),(Item=(Type=Bone,Name=\"deform_upperarm_twist_01_l\"),Transform=(Rotation=(X=0,Y=0,Z=0,W=1),Translation= ...(truncated)"
    WeightMinimum=0
    Weight=@n11.Result
   )
   > &RigUnit_ModifyTransforms(
    ItemToModify=((Item=(Type=Bone,Name="clavicle_extra_ASW_l"),Transform=(Rotation=(X=0,Y=0.0756,Z=0,W=0.9971),Translation=(X=32.7793,Y=-11.6185,Z=-17.6461),Scale3D=(X=1,Y=1,Z=2.104))))
    WeightMinimum=0
    Weight=&RigVMFunction_MathDoubleClamp(Maximum=0.7, Value=@n11.Result).Result
   )
   > n12 = &RigUnit_SphericalPoseReader(
    ActiveRegionScaleFactors=(PositiveWidth=1,NegativeWidth=1,PositiveHeight=1,NegativeHeight=1)
    ActiveRegionSize=0.2
    Debug=(bDrawDebug=true,bDraw2D=False,bDrawLocalAxes=False,DebugScale=25,DebugSegments=20,DebugThickness=0.25)
    DriverAxis=(X=-1,Y=0,Z=0)
    DriverItem=(Type=Bone,Name="upperarm_r")
    FalloffRegionScaleFactors=(PositiveWidth=1,NegativeWidth=1,PositiveHeight=1,NegativeHeight=1)
    FalloffSize=0.45
    FlipHeightScaling=true
    FlipWidthScaling=false
    OptionalParentItem=(Type=Bone,Name="pelvis")
    RotationOffset=(X=-30,Y=40,Z=0)
   )
   n13 = &RigVMFunction_AlphaInterp(InterpSpeedDecreasing=5, InterpSpeedIncreasing=5, Scale=1, bClampResult=true, Value=RigVMVariableNode() {variable=FrontEndValue}.Value)
   n14 = &RigVMFunction_MathDoubleMul(B=@n13.Result, A=@n12.OutputParam)
   > &RigUnit_ModifyTransforms(
    ItemToModify="((Item=(Type=Bone,Name=\"deform_upperarm_up_r\"),Transform=(Rotation=(X=0,Y=-0.2835,Z=0,W=0.959),Translation=(X=21.8434,Y=11.5073,Z=-11.0747),Scale3D=(X=1.5,Y=2,Z=1))),(Item=(Type=Bone,Name=\"deform_upperarm_twist_01_r\"),Transform=(Rotation=(X=0,Y=0.0213,Z=0, ...(truncated)"
    WeightMinimum=0
    Weight=@n14.Result
   )
   > &RigUnit_ModifyTransforms(
    ItemToModify=((Item=(Type=Bone,Name="clavicle_extra_ASW_r"),Transform=(Rotation=(X=0,Y=-0.4708,Z=0,W=-0.8822),Translation=(X=-30.5961,Y=3.4738,Z=19.6363),Scale3D=(X=1.0195,Y=1,Z=3.7058))))
    WeightMaximum=0.7
    WeightMinimum=0
    Weight=&RigVMFunction_MathDoubleClamp(Maximum=0.55, Value=@n14.Result).Result
   )
   RigVMFunctionReturnNode(ExecuteContext=@RigVMFunction_Sequence.C)
   RigUnit_GetTransform = &RigUnit_GetTransform(Item=(Type=Bone,Name="hand_l"), bInitial=False)
   RigUnit_SetTransform = &RigUnit_SetTransform(Item=(Type=Bone,Name="deform_hand_l"), Weight=1, bInitial=False, Value=@RigUnit_GetTransform.Transform, ExecutePin=@RigVMFunction_Sequence.B)
   > &RigUnit_SetTransform(
    Item=(Type=Bone,Name="deform_hand_r")
    Weight=1
    bInitial=False
    Value=&RigUnit_GetTransform(Item=(Type=Bone,Name="hand_r"), bInitial=False).Transform
   )
   &RigUnit_ModifyTransforms(
    ExecutePin=()
    ItemToModify=((Item=(Type=Bone,Name="clavicle_extra_ASW_r"),Transform=(Rotation=(X=0,Y=-0.4079,Z=0,W=-0.913),Translation=(X=-36.9572,Y=3.474,Z=28.0189),Scale3D=(X=1.0195,Y=1,Z=3.7058))))
    Weight=1
    WeightMaximum=0.7
    WeightMinimum=0
   )
  ]}
 New Function_ContainedGraph {nodes [
   n1 = RigVMFunctionEntryNode()
   n2 = &RigUnit_GetTransform(bInitial=False, Item=@n1.Start)
   Subtract = /Script/RigVM.RigVMFunction_MathVectorSub(B=@n2.Transform.Translation, A=&RigUnit_GetTransform(bInitial=False, Item=@n1.End).Transform.Translation) {is=struct RigVMUnitNode, template="Subtract::Execute(in A,in B,out Result)"}
   Unit = /Script/RigVM.RigVMFunction_MathVectorUnit(Value=@Subtract.Result) {is=struct RigVMUnitNode, template="Unit::Execute(in Value,out Result)"}
   RigVMFunction_MathQuaternionFromAxisAndAngle = /Script/RigVM.RigVMFunction_MathQuaternionFromAxisAndAngle(Axis=@Unit.Result, Angle=@n1.QuatAmount) {is=struct RigVMUnitNode, template="FromAxisAndAngle::Execute(in Axis,in Angle,out Result)"}
   Multiply = /Script/RigVM.RigVMFunction_MathQuaternionMul(B=@n2.Transform.Rotation, A=@RigVMFunction_MathQuaternionFromAxisAndAngle.Result) {is=struct RigVMUnitNode, template="Multiply::Execute(in A,in B,out Result)"}
   > &RigUnit_SetRotation(Space=GlobalSpace, bInitial=False, Item=@n1.Start, Weight=@n1.Weight, Value=@Multiply.Result)
   > RigVMFunctionReturnNode()
  ]}
 New Function_ContainedGraph {nodes [
   RigVMFunctionEntryNode()
   > RigVMFunction_Sequence = RigVMAggregateNode()
   RigVMFunctionReturnNode(ExecuteContext=@RigVMFunction_Sequence.E)
   n1 = /Script/RigVM.RigVMFunction_MathQuaternionToAxisAndAngle(Value=&RigVMFunction_MathQuaternionSwingTwist(
     TwistAxis=(X=1,Y=0,Z=0)
     Input=&RigUnit_GetTransform(Item=(Type=Bone,Name="deform_upperarm_r"), Space=LocalSpace, bInitial=False).Transform.Rotation
    ).Twist) {is=struct RigVMUnitNode, template="ToAxisAndAngle::Execute(in Value,out Axis,out Angle)"}
   &RigUnit_SetRotation(
    Item=(Type=Bone,Name="deform_upperarm_twist_01_r")
    Weight=1
    bInitial=False
    Value=&RigVMFunction_MathQuaternionFromAxisAndAngle(Angle=&RigVMFunction_MathDoubleMul(B=-0.6, A=@n1.Angle).Result, Axis=@n1.Axis) {template="FromAxisAndAngle::Execute(in Axis,in Angle,out Result)"}.Result
    ExecutePin=@RigVMFunction_Sequence.B
   )
   > &RigUnit_SetRotation(
    Item=(Type=Bone,Name="deform_upperarm_twist_02_r")
    Weight=1
    bInitial=False
    Value=&RigVMFunction_MathQuaternionFromAxisAndAngle(Angle=&RigVMFunction_MathDoubleMul(B=-0.2, A=@n1.Angle).Result, Axis=@n1.Axis) {template="FromAxisAndAngle::Execute(in Axis,in Angle,out Result)"}.Result
   )
   > &RigUnit_SetRotation(
    Item=(Type=Bone,Name="deform_lowerarm_twist_01_r")
    Weight=1
    bInitial=False
    Value=&RigVMFunction_MathQuaternionSwingTwist(
     TwistAxis=(X=1,Y=0,Z=0)
     Input=&RigUnit_GetTransform(Item=(Type=Bone,Name="lowerarm_twist_01_r"), Space=LocalSpace, bInitial=False).Transform.Rotation
    ).Twist
   )
   > &RigUnit_SetRotation(
    Item=(Type=Bone,Name="deform_lowerarm_twist_02_r")
    Weight=0.3
    bInitial=False
    Value=&RigVMFunction_MathQuaternionSwingTwist(
     TwistAxis=(X=1,Y=0,Z=0)
     Input=&RigUnit_GetTransform(Item=(Type=Bone,Name="lowerarm_twist_02_r"), Space=LocalSpace, bInitial=False).Transform.Rotation
    ).Swing
   )
   RigUnit_GetTransform = &RigUnit_GetTransform(Item=(Type=Bone,Name="deform_upperarm_l"), Space=LocalSpace, bInitial=False)
   RigVMFunction_MathQuaternionSwingTwist = &RigVMFunction_MathQuaternionSwingTwist(TwistAxis=(X=-1,Y=0,Z=0), Input=@RigUnit_GetTransform.Transform.Rotation)
   RigVMFunction_MathQuaternionToAxisAndAngle = &RigVMFunction_MathQuaternionToAxisAndAngle(Value=@RigVMFunction_MathQuaternionSwingTwist.Twist) {template="ToAxisAndAngle::Execute(in Value,out Axis,out Angle)"}
   RigVMFunction_MathQuaternionFromAxisAndAngle = &RigVMFunction_MathQuaternionFromAxisAndAngle(
    Angle=&RigVMFunction_MathDoubleMul(B=-0.6, A=@RigVMFunction_MathQuaternionToAxisAndAngle.Angle).Result
    Axis=@RigVMFunction_MathQuaternionToAxisAndAngle.Axis
   ) {template="FromAxisAndAngle::Execute(in Axis,in Angle,out Result)"}
   &RigUnit_SetRotation(
    Item=(Type=Bone,Name="deform_upperarm_twist_01_l")
    Weight=1
    bInitial=False
    Value=@RigVMFunction_MathQuaternionFromAxisAndAngle.Result
    ExecutePin=@RigVMFunction_Sequence.A
   )
   > &RigUnit_SetRotation(
    Item=(Type=Bone,Name="deform_upperarm_twist_02_l")
    Weight=1
    bInitial=False
    Value=&RigVMFunction_MathQuaternionFromAxisAndAngle(
     Angle=&RigVMFunction_MathDoubleMul(B=-0.2, A=@RigVMFunction_MathQuaternionToAxisAndAngle.Angle).Result
     Axis=@RigVMFunction_MathQuaternionToAxisAndAngle.Axis
    ) {template="FromAxisAndAngle::Execute(in Axis,in Angle,out Result)"}.Result
   )
   n2 = &RigUnit_GetTransform(Item=(Type=Bone,Name="lowerarm_twist_01_l"), Space=LocalSpace, bInitial=False)
   > &RigUnit_SetRotation(Item=(Type=Bone,Name="deform_lowerarm_twist_01_l"), Weight=1, bInitial=False, Value=@n2.Transform.Rotation)
   n3 = &RigUnit_GetTransform(Item=(Type=Bone,Name="lowerarm_twist_02_l"), Space=LocalSpace, bInitial=False)
   > &RigUnit_SetRotation(Item=(Type=Bone,Name="deform_lowerarm_twist_02_l"), Weight=0.3, bInitial=False, Value=@n3.Transform.Rotation)
   RigVMRerouteNode(Value=@RigVMFunction_Sequence.D)
   RigVMRerouteNode(Value=@RigVMFunction_Sequence.C)
   n4 = &RigUnit_GetTransform(Item=(Type=Bone,Name="clavicle_l"), Space=LocalSpace, bInitial=False)
   n5 = &RigUnit_GetTransform(Item=(Type=Bone,Name="clavicle_l"), Space=LocalSpace, bInitial=true)
   n6 = &RigVMFunction_MathQuaternionToAxisAndAngle(Value=&RigVMFunction_MathQuaternionSwingTwist(
     TwistAxis=(X=1,Y=0,Z=0)
     Input=&RigVMFunction_MathTransformMakeRelative(Parent=@n4.Transform, Global=@n5.Transform).Local.Rotation
    ).Twist) {template="ToAxisAndAngle::Execute(in Value,out Axis,out Angle)"}
   &RigVMFunction_MathQuaternionFromAxisAndAngle(Angle=&RigVMFunction_MathDoubleMul(B=-1, A=@n6.Angle).Result, Axis=@n6.Axis) {template="FromAxisAndAngle::Execute(in Axis,in Angle,out Result)"}
   n7 = &RigUnit_GetTransform(Item=(Type=Bone,Name="clavicle_r"), Space=LocalSpace, bInitial=true)
   n8 = &RigUnit_GetTransform(Item=(Type=Bone,Name="clavicle_r"), Space=LocalSpace, bInitial=False)
   n9 = &RigVMFunction_MathQuaternionToAxisAndAngle(Value=&RigVMFunction_MathQuaternionSwingTwist(
     TwistAxis=(X=1,Y=0,Z=0)
     Input=&RigVMFunction_MathTransformMakeRelative(Global=@n7.Transform, Parent=@n8.Transform).Local.Rotation
    ).Twist) {template="ToAxisAndAngle::Execute(in Value,out Axis,out Angle)"}
   &RigVMFunction_MathQuaternionFromAxisAndAngle(Angle=&RigVMFunction_MathDoubleMul(B=-1, A=@n9.Angle).Result, Axis=@n9.Axis) {template="FromAxisAndAngle::Execute(in Axis,in Angle,out Result)"}
   &RigVMFunction_MathQuaternionSwingTwist(TwistAxis=(X=1,Y=0,Z=0), Input=@n2.Transform.Rotation)
   &RigVMFunction_MathQuaternionSwingTwist(TwistAxis=(X=1,Y=0,Z=0), Input=@n3.Transform.Rotation)
  ]}
 New Function_ContainedGraph {nodes [
   RigVMFunctionEntryNode()
   > n1 = &RigUnit_SphericalPoseReader(
    ActiveRegionScaleFactors=(PositiveWidth=0.5,NegativeWidth=0.5,PositiveHeight=1,NegativeHeight=1)
    ActiveRegionSize=0.05
    Debug=(bDrawDebug=false,DebugScale=25,DebugSegments=20,DebugThickness=0.25)
    DriverAxis=(X=0,Y=1,Z=0)
    DriverItem=(Type=Bone,Name="head")
    FalloffRegionScaleFactors=(PositiveWidth=0.3,NegativeWidth=0.3,PositiveHeight=0.5,NegativeHeight=0.5)
    FalloffSize=0.6
    OptionalParentItem=(Type=Bone,Name="spine_03")
    RotationOffset=(X=90,Y=0,Z=35)
   )
   > &RigUnit_ModifyTransforms(
    ItemToModify=((Item=(Type=Bone,Name="extra_neck"),Transform=(Rotation=(X=0,Y=0,Z=0,W=1),Translation=(X=0,Y=0,Z=-4),Scale3D=(X=1,Y=1,Z=1))))
    WeightMinimum=0
    Weight=@n1.OutputParam
   )
   AlphaInterp = &RigVMFunction_AlphaInterp(InterpSpeedDecreasing=10, InterpSpeedIncreasing=10, Scale=1, bClampResult=true, Value=RigVMVariableNode() {variable=Idle FE}.Value)
   > &RigUnit_ModifyTransforms(
    ItemToModify=((Item=(Type=Bone,Name="extra_neck"),Transform=(Rotation=(X=0,Y=0,Z=0,W=1),Translation=(X=0,Y=0,Z=-4),Scale3D=(X=1,Y=1,Z=1))))
    WeightMinimum=0
    Weight=@AlphaInterp.Result
   )
   > RigVMFunctionReturnNode()
  ]}
 New Function_ContainedGraph {nodes [
   n1 = RigVMFunctionEntryNode()
   Clamp = &RigVMFunction_MathIntClamp(Value=@n1.primAxis)
   RigVMFunction_MathBoolToInteger = &RigVMFunction_MathBoolToInteger(Value=@n1.onlyTwist)
   RigVMFunction_MathQuaternionSwingTwist = &RigVMFunction_MathQuaternionSwingTwist(
    TwistAxis=@n1.twistAxis
    Input=&RigUnit_GetRelativeTransformForItem(
     Child=(Type=Bone,Name="None")
     Parent=(Type=Bone,Name="None")
     bChildInitial=False
     bParentInitial=False
     Child.Name=@n1.child
     Parent.Name=@n1.parent
     Parent.Type=@n1.parentType
    ).RelativeTransform.Rotation
   )
   RigVMFunction_MathQuaternionToEuler = &RigVMFunction_MathQuaternionToEuler(
    Value=RigVMDispatchNode(
     Values=((),())
     Index=@RigVMFunction_MathBoolToInteger.Result
     Values.0=@RigVMFunction_MathQuaternionSwingTwist.Swing
     Values.1=@RigVMFunction_MathQuaternionSwingTwist.Twist
    ) {template="DISPATCH_RigVMDispatch_SelectInt32(in Index,in Values,out Result)"}.Result
    RotationOrder=@n1.rotationOrder
   )
   DISPATCH_RigVMDispatch_SelectInt32 = RigVMDispatchNode(
    Values=(0,0,0)
    Index=@Clamp.Result
    Values.0=@RigVMFunction_MathQuaternionToEuler.Result.X
    Values.1=@RigVMFunction_MathQuaternionToEuler.Result.Y
    Values.2=@RigVMFunction_MathQuaternionToEuler.Result.Z
   ) {template="DISPATCH_RigVMDispatch_SelectInt32(in Index,in Values,out Result)"}
   Remap = &RigVMFunction_MathDoubleRemap(
    TargetMaximum=1
    TargetMinimum=0
    bClamp=False
    Value=@DISPATCH_RigVMDispatch_SelectInt32.Result
    SourceMaximum=@n1.max
    SourceMinimum=@n1.min
   )
   > RigVMFunctionReturnNode(OuputWeight=&RigVMFunction_MathDoubleClamp(Value=@Remap.Result).Result)
  ]}
 RigVMFunctionLibrary {nodes [
   Adjust HeadAndNeck = RigVMCollapseNode(ExecuteContext=())
   ApplyFBIKCahinCorrections = RigVMCollapseNode(End=(Type=None,Name="None"), ExecuteContext=(), Start=(Type=None,Name="None"))
   ArmIKCorrections = RigVMCollapseNode(DisableArmsModifications=False, ExecuteContext=())
   ControlsSetup = RigVMCollapseNode(Controls=False, ExecuteContext=())
   DeformTwist = RigVMCollapseNode(ExecuteContext=())
   DeltaRotation = RigVMCollapseNode(ExecuteContext=(), debugRotation=False, onlyTwist=False, twistAxis=(X=0,Y=0,Z=0))
   DeltaRotation Z = RigVMCollapseNode(ExecuteContext=())
   RobusTornIK = RigVMCollapseNode(Disable Arms=False, ExecuteContext=())
   SplineLeg = RigVMCollapseNode(ExecuteContext=())
  ]}
 RigVMFunction_Sequence_1_1_ContainedGraph {nodes [
   n1 = RigVMFunctionEntryNode()
   n2 = &RigVMFunction_Sequence(ExecuteContext=@n1.ExecuteContext)
   n3 = &RigVMFunction_Sequence(ExecuteContext=@n2.B)
   RigVMFunctionReturnNode(A=@n2.A, B=@n3.A, C=@n3.B)
  ]}
 RigVMFunction_Sequence_1_ContainedGraph {nodes [
   n1 = RigVMFunctionEntryNode()
   RigVMFunction_Sequence = &RigVMFunction_Sequence(ExecuteContext=@n1.ExecuteContext)
   n2 = &RigVMFunction_Sequence(ExecuteContext=@RigVMFunction_Sequence.B)
   n3 = &RigVMFunction_Sequence(ExecuteContext=@n2.B)
   n4 = &RigVMFunction_Sequence(ExecuteContext=@n3.B)
   RigVMFunctionReturnNode(A=@RigVMFunction_Sequence.A, B=@n2.A, C=@n3.A, D=@n4.A, E=@n4.B)
  ]}
 RigVMFunction_Sequence_1_ContainedGraph {nodes [
   n1 = RigVMFunctionEntryNode()
   RigVMFunction_Sequence = &RigVMFunction_Sequence(ExecuteContext=@n1.ExecuteContext)
   n2 = &RigVMFunction_Sequence(ExecuteContext=@RigVMFunction_Sequence.B)
   n3 = &RigVMFunction_Sequence(ExecuteContext=@n2.B)
   n4 = &RigVMFunction_Sequence(ExecuteContext=@n3.B)
   RigVMFunctionReturnNode(A=@RigVMFunction_Sequence.A, B=@n2.A, C=@n3.A, D=@n4.A, E=@n4.B)
  ]}
 RigVMFunction_Sequence_2_ContainedGraph {nodes [
   n1 = RigVMFunctionEntryNode()
   RigVMFunction_Sequence = &RigVMFunction_Sequence(ExecuteContext=@n1.ExecuteContext)
   n2 = &RigVMFunction_Sequence(ExecuteContext=@RigVMFunction_Sequence.B)
   RigVMFunctionReturnNode(A=@RigVMFunction_Sequence.A, B=@n2.A, C=@n2.B)
  ]}
 RigVMModel {nodes [
   RigUnit_BeginExecution = &RigUnit_BeginExecution()
   ControlsSetup = RigVMFunctionReferenceNode(Controls=false, ExecuteContext=@RigUnit_BeginExecution.ExecutePin)
   > FN_DeformRig_FNC = RigVMFunctionReferenceNode(
    DeformRig_ Curve="(deformRig_l_upperarm_up_alpha_72_#3=true,deformRig_r_upperarm_up_alpha_73_#4=true,deformRig_l_upperarm_fr_alpha_74_#5=False,deformRig_r_upperarm_fr_alpha_75_CB51B80C ...(truncated)"
    DeformRig_Bool="(upperarm_1_#6=true,elbow_3_#7=false,wrist_7_#8=false,groin_8_#9=true,glute_17_#10=true,knee_9_40C408324CA2 ...(truncated)"
    DeformRig_Debug=(upperarm_debug_18_#6=False,wrist_debug_20_#8=False,lat_debug_26_#11=False,groinGlute_debug_27_#9=False)
    DeformRig_Value="(upperarm_up_18_#1=(Rotation=(X=0,Y=0,Z=0,W=1),Translation=(X=0,Y=0,Z=8),Scale3D=(X=0.7,Y=1,Z=1)),upperarm_fwd_fr_19_#2=(Rotation=(X=0,Y=0,Z=0,W=1),Translation=(X=0,Y=10,Z=0),Scale3D=(X=0.5,Y=1,Z= ...(truncated)"
   )
   > SplineLeg = RigVMFunctionReferenceNode()
   > RobusTornIK = RigVMFunctionReferenceNode(Disable Arms=false)
   > Adjust HeadAndNeck = RigVMFunctionReferenceNode()
   &RigUnit_PrepareForExecution()
  ]}
]
hierarchy [
 root {loc=<0 0 0>, rot=<0 0 0 1>}
 pelvis {parent=root, loc=<0 -0.7321 92.4218>, rot=<0.0127 0.707 -0.0127 -0.707>}
 spine_01 {parent=pelvis, loc=<5.6002 -0.5128 0>, rot=<0 0 0.034 -0.9994>}
 spine_02 {parent=spine_01, loc=<8.0756 0.0139 0>, rot=<0 0 0.0238 -0.9997>}
 spine_03 {parent=spine_02, loc=<7.0395 0 0>, rot=<0 0 0.0149 -0.9999>}
 spine_04 {parent=spine_03, loc=<8.1023 0 0>, rot=<0 0 -0.0132 -0.9999>}
 spine_05 {parent=spine_04, loc=<10.0141 0 0>, rot=<0 0 -0.0015 -1>}
 clavicle_l {parent=spine_05, loc=<2.5856 -0.2866 -4.3777>, rot=<0.0465 -0.7289 -0.009 -0.6829>}
 upperarm_l {parent=clavicle_l, loc=<18.2264 0.1342 -0.001>, rot=<0.0253 -0.2693 0.0013 -0.9627>}
 lowerarm_l {parent=upperarm_l, loc=<29.0071 0.1686 -1.239>, rot=<0.0391 -0.0099 -0.2453 -0.9686>}
 hand_l {parent=lowerarm_l, loc=<25.8607 0.222 1.0274>, rot=<-0.7193 -0.0988 0.1193 -0.6773>}
 thumb_01_l {parent=hand_l, loc=<1.3118 -0.3191 -2.459>, rot=<0.6546 -0.4042 0.0664 -0.6354>}
 thumb_02_l {parent=thumb_01_l, loc=<5.9336 -0.0448 0.0401>, rot=<-0.0037 0.0213 0.298 -0.9543>}
 thumb_03_l {parent=thumb_02_l, loc=<2.2223 0.3682 -0.0907>, rot=<0.0118 -0.013 0.0066 -0.9998>}
 index_metacarpal_l {parent=hand_l, loc=<1.6535 0.0183 -2.96>, rot=<0.0158 -0.0124 -0.0173 -0.9996>}
 index_01_l {parent=index_metacarpal_l, loc=<8.3008 0.0357 -0.0524>, rot=<-0.0044 0.0485 0.0924 -0.9945>}
 index_02_l {parent=index_01_l, loc=<3.3587 -0.3004 0.0452>, rot=<-0.002 0.0027 0.3225 -0.9466>}
 index_03_l {parent=index_02_l, loc=<2.5578 -0.123 -0.0252>, rot=<-0.002 0.0067 0.2558 -0.9667>}
 middle_metacarpal_l {parent=hand_l, loc=<2.1856 -0.1725 -0.5984>, rot=<0.006 0.0168 -0.0329 -0.9993>}
 middle_01_l {parent=middle_metacarpal_l, loc=<7.3275 -0.2447 0.0129>, rot=<-0.0208 0.0042 0.0858 -0.9961>}
 middle_02_l {parent=middle_01_l, loc=<4.0123 0.1401 0.1175>, rot=<-0.0218 0.0258 0.3512 -0.9357>}
 middle_03_l {parent=middle_02_l, loc=<2.6077 0.0534 0.1373>, rot=<0.0082 0.0188 0.1708 -0.9851>}
 ring_metacarpal_l {parent=hand_l, loc=<2.2074 -0.3625 1.3638>, rot=<-0.0284 0.0722 -0.0458 -0.9959>}
 ring_01_l {parent=ring_metacarpal_l, loc=<7.0295 -0.6197 -0.1446>, rot=<-0.0489 -0.0252 0.1892 -0.9804>}
 ring_02_l {parent=ring_01_l, loc=<3.6174 0.4647 -0.1027>, rot=<0 0 0.2806 -0.9598>}
 ring_03_l {parent=ring_02_l, loc=<2.4152 0.114 0.0199>, rot=<0 0 0.1189 -0.9929>}
 pinky_metacarpal_l {parent=hand_l, loc=<1.7616 -0.9753 3.3814>, rot=<-0.0397 0.0814 -0.069 -0.9935>}
 pinky_01_l {parent=pinky_metacarpal_l, loc=<7.1709 -0.4831 -0.1094>, rot=<-0.0654 -0.0035 0.2242 -0.9723>}
 pinky_02_l {parent=pinky_01_l, loc=<2.8274 0.0893 -0.1043>, rot=<0.0037 0.0011 0.2854 -0.9584>}
 pinky_03_l {parent=pinky_02_l, loc=<2.2046 0.0307 -0.0078>, rot=<-0.0002 0.0098 0.1445 -0.9895>}
 deform_wrist_dn_base_l {parent=hand_l, loc=<0 -3.4173 0>, rot=<0 0 0 -1>}
 deform_wrist_dn_l {parent=deform_wrist_dn_base_l, loc=<0 0 0>, rot=<0 0 0 1>}
 deform_wrist_up_base_l {parent=hand_l, loc=<0 2.7805 0>, rot=<0 0 0 -1>}
 deform_wrist_up_l {parent=deform_wrist_up_base_l, loc=<0 0 0>, rot=<0 0 0 1>}
 weapon_l {parent=hand_l, loc=<-0.1974 -3.4111 0.3811>, rot=<0 0 0 1>}
 hand_attach_l {parent=hand_l, loc=<8.5 -3 0>, rot=<0.7071 0.7071 0 0>}
 lowerarm_twist_02_l {parent=lowerarm_l, loc=<9.6199 0 0>, rot=<0 0 0 -1>}
 lowerarm_twist_01_l {parent=lowerarm_l, loc=<17.8549 0 0>, rot=<0 0 0 -1>}
 upperarm_twist_01_l {parent=upperarm_l, loc=<7.9796 0.0843 -0.6195>, rot=<0 0 0 -1>}
 upperarm_twist_02_l {parent=upperarm_l, loc=<19.5059 0.0843 -0.6195>, rot=<0 0 0 -1>}
 deform_upperarm_l {parent=clavicle_l, loc=<18.2264 0.1342 -0.001>, rot=<0.0253 -0.2693 0.0013 -0.9627>}
 deform_lowerarm_l {parent=deform_upperarm_l, loc=<29.0071 0.1686 -1.239>, rot=<0.0391 -0.0099 -0.2453 -0.9686>}
 deform_hand_l {parent=deform_lowerarm_l, loc=<25.8607 0.222 1.0274>, rot=<-0.7193 -0.0988 0.1193 -0.6773>}
 deform_elbow_in_base_l {parent=deform_lowerarm_l, loc=<0.2112 5.645 0.8274>, rot=<0 0 0 1>}
 deform_elbow_in_l {parent=deform_elbow_in_base_l, loc=<0 0 0>, rot=<0 0 0 1>}
 deform_elbow_out_base_l {parent=deform_lowerarm_l, loc=<0 -7.5129 0>, rot=<0 0 0 1>}
 deform_elbow_out_l {parent=deform_elbow_out_base_l, loc=<0 0 0>, rot=<0 0 0 1>}
 deform_lowerarm_twist_02_l {parent=deform_lowerarm_l, loc=<9.6199 0 0>, rot=<0 0 0 -1>}
 deform_lowerarm_twist_01_l {parent=deform_lowerarm_l, loc=<17.8549 0 0>, rot=<0 0 0 -1>}
 deform_upperarm_twist_01_l {parent=deform_upperarm_l, loc=<7.9796 0.0843 -0.6195>, rot=<0 0 0 -1>}
 deform_upperarm_bk_base_l {parent=deform_upperarm_twist_01_l, loc=<-6.8983 -13.3826 -1.3258>, rot=<0 0 0 -1>}
 deform_upperarm_bk_l {parent=deform_upperarm_bk_base_l, loc=<0 0 0>, rot=<0 0 0 1>}
 deform_upperarm_fr_base_l {parent=deform_upperarm_twist_01_l, loc=<-8.3487 10.7117 -2.1455>, rot=<0 0 0 -1>}
 deform_upperarm_fr_l {parent=deform_upperarm_fr_base_l, loc=<0 0 0>, rot=<0 0 0 1>}
 deform_upperarm_up_base_l {parent=deform_upperarm_twist_01_l, loc=<-11.2012 -1.2234 8.7784>, rot=<0 0 0 -1>}
 deform_upperarm_up_l {parent=deform_upperarm_up_base_l, loc=<0 0 0>, rot=<0 0 0 1>}
 deform_upperarm_twist_02_l {parent=deform_upperarm_l, loc=<19.5059 0.0843 -0.6195>, rot=<0 0 0 -1>}
 deform_bicep_base_l {parent=deform_upperarm_twist_02_l, loc=<0 8.0494 0>, rot=<0 0 0 1>}
 deform_bicep_l {parent=deform_bicep_base_l, loc=<0 0 0>, rot=<0 0 0 1>}
 upperarm_l_null {parent=clavicle_l, type=Null, loc=<18.2264 0.1342 -0.001>, rot=<0.0253 -0.2693 0.0013 -0.9627>}
 clavicle_extra_ASW_l {parent=clavicle_l, loc=<15.1415 2.7369 11.3217>, rot=<-0.0172 0.7302 -0.0184 -0.6828>}
 clavicle_parent_extra_l {parent=clavicle_l, loc=<15.1415 2.7369 11.3217>, rot=<-0.0172 0.7302 -0.0184 -0.6828>}
 clavicle_extra_l {parent=clavicle_parent_extra_l, loc=<0 0 0>, rot=<0 0 0 1>}
 neck_01 {parent=spine_05, loc=<10.815 1.9196 0>, rot=<0 0 -0.1864 -0.9825>}
 neck_02 {parent=neck_01, loc=<4.7347 0.3045 0>, rot=<0 0 0.0363 -0.9993>}
 head {parent=neck_02, loc=<5.7632 0.5552 0>, rot=<0 0 0.115 -0.9934>}
 faceAttach {parent=head, loc=<2.0755 -0.8031 0>, rot=<0.0029 -0.7071 -0.0029 -0.7071>}
 C_brow_mid {parent=faceAttach, loc=<0 7.0458 7.3468>, rot=<0 0 0 -1>}
 L_brow_mid {parent=faceAttach, loc=<1.9272 6.8384 7.3251>, rot=<-0.0011 -0.0041 0.018 -0.9998>}
 L_brow_outer {parent=faceAttach, loc=<4.4883 6.8168 7.2218>, rot=<-0.0408 -0.0274 0.1076 -0.993>}
 L_cheek_inner {parent=faceAttach, loc=<5.7775 7.2115 0.3277>, rot=<-0.0706 -0.2697 0.3184 -0.906>}
 C_lip_upper_mid {parent=faceAttach, loc=<0.0022 12.636 -2.1487>, rot=<-0.0217 0.0012 -0.0008 -0.9998>}
 L_lip_upper_outer {parent=faceAttach, loc=<3.2317 11.9368 -2.1923>, rot=<-0.0207 0.0114 0.1836 -0.9827>}
 L_lip_corner {parent=faceAttach, loc=<6.3423 9.8034 -2.8185>, rot=<0.0053 0.0067 -0.3197 -0.9475>}
 R_brow_mid {parent=faceAttach, loc=<-1.9272 6.8384 7.3251>, rot=<-0.0011 0.0041 -0.018 -0.9998>}
 R_brow_outer {parent=faceAttach, loc=<-4.4883 6.8168 7.2218>, rot=<-0.0408 0.0274 -0.1076 -0.993>}
 R_cheek_inner {parent=faceAttach, loc=<-5.7775 7.2115 0.3277>, rot=<-0.0706 0.2697 -0.3184 -0.906>}
 R_lip_corner {parent=faceAttach, loc=<-6.3423 9.8034 -2.8185>, rot=<0.0053 -0.0067 0.3197 -0.9475>}
 C_jaw {parent=faceAttach, loc=<-0.0134 -0.1613 -1.4291>, rot=<0 0 0 1>}
 L_lip_lower_outer {parent=C_jaw, loc=<3.3027 12.133 -1.6085>, rot=<0.0006 0.0116 0.1669 -0.9859>}
 C_lip_lower_mid {parent=C_jaw, loc=<0.0223 13.204 -1.4818>, rot=<0 0 0 -1>}
 R_lip_lower_outer {parent=C_jaw, loc=<-3.2759 12.133 -1.6085>, rot=<0.0006 -0.0116 -0.1669 -0.9859>}
 teeth_lower {parent=C_jaw, loc=<0.0134 10.3789 -1.6905>, rot=<0 0 0 1>}
 tongue {parent=C_jaw, loc=<0.0141 3.4603 -0.4656>, rot=<-0.1568 0 0 -0.9876>}
 R_lip_upper_outer {parent=faceAttach, loc=<-3.2317 11.9368 -2.1923>, rot=<-0.0207 -0.0114 -0.1836 -0.9827>}
 C_nose_bridge {parent=faceAttach, loc=<0 8.1847 2.7063>, rot=<-0.2218 0 0 -0.9751>}
 R_eye {parent=faceAttach, loc=<2.6294 5.3535 4.651>, rot=<0 0 0.9998 -0.0175>}
 R_eye_lid_upper_mid {parent=R_eye, loc=<0 0 0>, rot=<0.0523 0 0 -0.9986>}
 R_eye_lid_lower_mid {parent=R_eye, loc=<0 0 0>, rot=<0.0329 0 0 -0.9995>}
 L_eye {parent=faceAttach, loc=<-2.6294 5.3535 4.6511>, rot=<0 0 0.9998 0.0175>}
 L_eye_lid_upper_mid {parent=L_eye, loc=<0 0 0>, rot=<0.0523 0 0 -0.9986>}
 L_eye_lid_lower_mid {parent=L_eye, loc=<0 0 0>, rot=<0.0329 0 0 -0.9995>}
 teeth_upper {parent=faceAttach, loc=<0 11.272 -2.1143>, rot=<0 0 0 1>}
 head_back {parent=head, loc=<3.6568 -11.1292 0>, rot=<0.5021 -0.4979 0.4979 -0.5021>}
 clavicle_r {parent=spine_05, loc=<2.5855 -0.2866 4.3777>, rot=<0.7289 0.0465 -0.6829 0.009>}
 upperarm_r {parent=clavicle_r, loc=<-18.2264 -0.1342 0.001>, rot=<0.0253 -0.2693 0.0013 -0.9627>}
 lowerarm_r {parent=upperarm_r, loc=<-29.0071 -0.1686 1.239>, rot=<0.0391 -0.0099 -0.2453 -0.9686>}
 hand_r {parent=lowerarm_r, loc=<-25.8608 -0.222 -1.0274>, rot=<-0.7193 -0.0988 0.1193 -0.6773>}
 thumb_01_r {parent=hand_r, loc=<-1.3118 0.3191 2.459>, rot=<0.6546 -0.4042 0.0664 -0.6354>}
 thumb_02_r {parent=thumb_01_r, loc=<-5.9336 0.0448 -0.0401>, rot=<-0.0037 0.0213 0.298 -0.9543>}
 thumb_03_r {parent=thumb_02_r, loc=<-2.2223 -0.3682 0.0907>, rot=<0.0118 -0.013 0.0066 -0.9998>}
 index_metacarpal_r {parent=hand_r, loc=<-1.6535 -0.0183 2.96>, rot=<0.0158 -0.0124 -0.0173 -0.9996>}
 index_01_r {parent=index_metacarpal_r, loc=<-8.3008 -0.0357 0.0524>, rot=<-0.0044 0.0485 0.0924 -0.9945>}
 index_02_r {parent=index_01_r, loc=<-3.3587 0.3004 -0.0452>, rot=<-0.002 0.0027 0.3225 -0.9466>}
 index_03_r {parent=index_02_r, loc=<-2.5578 0.123 0.0252>, rot=<-0.002 0.0067 0.2558 -0.9667>}
 middle_metacarpal_r {parent=hand_r, loc=<-2.1856 0.1725 0.5984>, rot=<0.006 0.0168 -0.0329 -0.9993>}
 middle_01_r {parent=middle_metacarpal_r, loc=<-7.3275 0.2447 -0.0129>, rot=<-0.0208 0.0042 0.0858 -0.9961>}
 middle_02_r {parent=middle_01_r, loc=<-4.0123 -0.1401 -0.1175>, rot=<-0.0218 0.0258 0.3512 -0.9357>}
 middle_03_r {parent=middle_02_r, loc=<-2.6077 -0.0534 -0.1373>, rot=<0.0082 0.0188 0.1708 -0.9851>}
 ring_metacarpal_r {parent=hand_r, loc=<-2.2074 0.3625 -1.3638>, rot=<-0.0284 0.0722 -0.0458 -0.9959>}
 ring_01_r {parent=ring_metacarpal_r, loc=<-7.0295 0.6197 0.1446>, rot=<-0.0489 -0.0252 0.1892 -0.9804>}
 ring_02_r {parent=ring_01_r, loc=<-3.6174 -0.4647 0.1027>, rot=<0 0 0.2806 -0.9598>}
 ring_03_r {parent=ring_02_r, loc=<-2.4152 -0.114 -0.0199>, rot=<0 0 0.1189 -0.9929>}
 pinky_metacarpal_r {parent=hand_r, loc=<-1.7616 0.9753 -3.3814>, rot=<-0.0397 0.0814 -0.069 -0.9935>}
 pinky_01_r {parent=pinky_metacarpal_r, loc=<-7.1709 0.4831 0.1094>, rot=<-0.0654 -0.0035 0.2242 -0.9723>}
 pinky_02_r {parent=pinky_01_r, loc=<-2.8274 -0.0893 0.1043>, rot=<0.0037 0.0011 0.2854 -0.9584>}
 pinky_03_r {parent=pinky_02_r, loc=<-2.2046 -0.0307 0.0078>, rot=<-0.0002 0.0098 0.1445 -0.9895>}
 deform_wrist_dn_base_r {parent=hand_r, loc=<0 3.4172 0>, rot=<0 0 0 -1>}
 deform_wrist_dn_r {parent=deform_wrist_dn_base_r, loc=<0 0 0>, rot=<0 0 0 -1>}
 deform_wrist_up_base_r {parent=hand_r, loc=<-0.0001 -2.7804 0>, rot=<0 0 0 -1>}
 deform_wrist_up_r {parent=deform_wrist_up_base_r, loc=<0 0 0>, rot=<0 0 0 -1>}
 hand_attach_r {parent=hand_r, loc=<-8.5 3 0>, rot=<0 0 -0.7071 -0.7071>}
 weapon_r {parent=hand_r, loc=<0.1974 3.4111 -0.3811>, rot=<0 0 0 1>}
 lowerarm_twist_02_r {parent=lowerarm_r, loc=<-9.6199 0 0>, rot=<0 0 0 -1>}
 lowerarm_twist_01_r {parent=lowerarm_r, loc=<-17.8549 0 0>, rot=<0 0 0 -1>}
 upperarm_twist_01_r {parent=upperarm_r, loc=<-7.9796 -0.0843 0.6195>, rot=<0 0 0 -1>}
 upperarm_twist_02_r {parent=upperarm_r, loc=<-19.5059 -0.0843 0.6195>, rot=<0 0 0 -1>}
 deform_upperarm_r {parent=clavicle_r, loc=<-18.2264 -0.1342 0.001>, rot=<0.0253 -0.2693 0.0013 -0.9627>}
 deform_lowerarm_r {parent=deform_upperarm_r, loc=<-29.0072 -0.1686 1.239>, rot=<0.0391 -0.0099 -0.2453 -0.9686>}
 deform_hand_r {parent=deform_lowerarm_r, loc=<-25.8608 -0.222 -1.0274>, rot=<-0.7193 -0.0988 0.1193 -0.6773>}
 deform_elbow_in_base_r {parent=deform_lowerarm_r, loc=<-0.2114 -5.6449 -0.827>, rot=<0 0 0 -1>}
 deform_elbow_in_r {parent=deform_elbow_in_base_r, loc=<0 0 0>, rot=<0 0 0 1>}
 deform_elbow_out_base_r {parent=deform_lowerarm_r, loc=<-0.0001 7.5129 0.0002>, rot=<0 0 0 -1>}
 deform_elbow_out_r {parent=deform_elbow_out_base_r, loc=<0 0 0>, rot=<0 0 0 1>}
 deform_lowerarm_twist_02_r {parent=deform_lowerarm_r, loc=<-9.6199 0 0>, rot=<0 0 0 -1>}
 deform_lowerarm_twist_01_r {parent=deform_lowerarm_r, loc=<-17.8549 0 0>, rot=<0 0 0 -1>}
 deform_upperarm_twist_01_r {parent=deform_upperarm_r, loc=<-7.9796 -0.0843 0.6195>, rot=<0 0 0 -1>}
 deform_upperarm_bk_base_r {parent=deform_upperarm_twist_01_r, loc=<6.8985 13.3826 1.3257>, rot=<0 0 0 -1>}
 deform_upperarm_bk_r {parent=deform_upperarm_bk_base_r, loc=<-0.0002 0 0.0002>, rot=<0 0 0 1>}
 deform_upperarm_fr_base_r {parent=deform_upperarm_twist_01_r, loc=<8.3485 -10.7117 2.1456>, rot=<0 0 0 -1>}
 deform_upperarm_fr_r {parent=deform_upperarm_fr_base_r, loc=<0.0001 0 -0.0001>, rot=<0 0 0 1>}
 deform_upperarm_up_base_r {parent=deform_upperarm_twist_01_r, loc=<11.2015 1.2234 -8.7788>, rot=<0 0 0 -1>}
 deform_upperarm_up_r {parent=deform_upperarm_up_base_r, loc=<-0.0003 0 0.0004>, rot=<0 0 0 1>}
 deform_upperarm_twist_02_r {parent=deform_upperarm_r, loc=<-19.5059 -0.0843 0.6195>, rot=<0 0 0 -1>}
 deform_bicep_base_r {parent=deform_upperarm_twist_02_r, loc=<0.0002 -8.0494 -0.0003>, rot=<0 0 0 1>}
 deform_bicep_r {parent=deform_bicep_base_r, loc=<0 0 0>, rot=<0 0 0 1>}
 upperarm_r_null {parent=clavicle_r, type=Null, loc=<-18.2264 -0.1342 0.001>, rot=<0.0253 -0.2693 0.0013 -0.9627>}
 clavicle_extra_ASW_r {parent=clavicle_r, loc=<-15.142 -2.737 -11.322>, rot=<-0.0172 0.7302 -0.0184 -0.6828>}
 clavicle_parent_extra_r {parent=clavicle_r, loc=<-15.142 -2.737 -11.322>, rot=<-0.0172 0.7302 -0.0184 -0.6828>}
 clavicle_extra_r {parent=clavicle_parent_extra_r, loc=<0.0002 0 -0.0005>, rot=<0 0 0 1>}
 attach_cape {parent=spine_05, loc=<9.0172 -11.7135 -0.0002>, rot=<0.0616 -0.7044 -0.0616 -0.7044>}
 cape {parent=attach_cape, type=Null, loc=<0 -6 0>, rot=<0 0 0 1>}
 attach_backpack {parent=spine_05, loc=<-0.1226 -13.472 -0.0001>, rot=<-0.0093 -0.7071 0.0093 -0.707>}
 Backpack_BR {parent=attach_backpack, type=Null, loc=<0 -7 0>, rot=<0 0 0 1>}
 extra_neck {parent=spine_05, loc=<8.9571 15.9775 0>, rot=<0.4795 0.5196 0.5196 0.4795>}
 deform_pec_aim_base_l {parent=spine_04, loc=<5.0227 13.6719 -2.6508>, rot=<0.0672 -0.5201 0.3228 -0.7879>}
 deform_pec_aim_l {parent=deform_pec_aim_base_l, loc=<0 0 0>, rot=<0 0 0 -1>}
 deform_pec_base_l {parent=deform_pec_aim_l, loc=<16.3279 0 0>, rot=<0 0 0 1>}
 deform_pec_l {parent=deform_pec_base_l, loc=<0 0 0>, rot=<0 0 0 1>}
 dyn_L_pectoralDynamics {parent=deform_pec_l, loc=<-5.8379 5.5543 -3.1702>, rot=<0 0 0 -1>}
 deform_pec_aim_base_r {parent=spine_04, loc=<5.0225 13.6718 2.6508>, rot=<-0.5201 -0.0672 0.7879 0.3228>}
 deform_pec_aim_r {parent=deform_pec_aim_base_r, loc=<0 0 0>, rot=<0 0 0 -1>}
 deform_pec_base_r {parent=deform_pec_aim_r, loc=<-16.3279 0 0>, rot=<0 0 0 -1>}
 deform_pec_r {parent=deform_pec_base_r, loc=<0 0 0.0002>, rot=<0 0 0 -1>}
 dyn_R_pectoralDynamics {parent=deform_pec_r, loc=<5.8377 -5.5545 3.1697>, rot=<0 0 0 1>}
 deform_lat_l_base {parent=spine_03, loc=<2.5561 -10.3716 -17.6918>, rot=<0 0 0 -1>}
 deform_lat_l {parent=deform_lat_l_base, loc=<0 0 0>, rot=<0 0 0 1>}
 deform_lat_r_base {parent=spine_03, loc=<2.5562 -10.3716 17.6918>, rot=<0 0 1 0>}
 deform_lat_r {parent=deform_lat_r_base, loc=<0 0 0>, rot=<0 0 0 -1>}
 thigh_l {parent=pelvis, loc=<-1.8716 -0.0109 -11.9936>, rot=<0.0012 0.0572 0.0155 -0.9982>}
 calf_l {parent=thigh_l, loc=<-41.9178 -0.2246 0.0926>, rot=<0.0006 -0.0161 -0.0391 -0.9991>}
 foot_l {parent=calf_l, loc=<-38.7057 1.2438 0.2452>, rot=<0.0975 -0.0474 0.0317 -0.9936>}
 ball_l {parent=foot_l, loc=<-6.1614 13.1469 0.6807>, rot=<0.0001 -0.0001 -0.7063 -0.7079>}
 dfrm_leg_spline_09_l_ctrl_null {parent=foot_l, type=Null, loc=<-0.0416 -0.0328 -0.1605>, rot=<-0.0487 -0.1016 0.9936 -0.0088>}
 dfrm_leg_spline_09_l_ctrl {controlType=EULER_TRANSFORM, parent=dfrm_leg_spline_09_l_ctrl_null, type=Control, loc=<0 0 0>, rot=<0 0 0 1>}
 calf_twist_02_l {parent=calf_l, loc=<-15.9268 -0.0967 0.0397>, rot=<0.0237 0.0095 -0.0035 -0.9997>}
 calf_twist_01_l {parent=calf_l, loc=<-32.9335 -0.0967 0.0397>, rot=<0.0237 0.0095 -0.0035 -0.9997>}
 deform_knee_in_base_l {parent=calf_l, loc=<1.298 -8.0356 -0.0496>, rot=<0 0 0 1>}
 deform_knee_in_l {parent=deform_knee_in_base_l, loc=<0 0 0>, rot=<0 0 0 1>}
 deform_knee_out_base_l {parent=calf_l, loc=<0 6.8739 0>, rot=<0 0 0 -1>}
 deform_knee_out_l {parent=deform_knee_out_base_l, loc=<0 0 0>, rot=<0 0 0 1>}
 deform_groin_base_l {parent=thigh_l, loc=<-2.1908 8.5661 -1.7442>, rot=<0 0 0 -1>}
 deform_groin_l {parent=deform_groin_base_l, loc=<0 0 0>, rot=<0 0 0 1>}
 thigh_twist_01_l {parent=thigh_l, loc=<-21.6303 0 0>, rot=<0.0013 0 0 -1>}
 thigh_r {parent=pelvis, loc=<-1.8714 -0.0109 11.9936>, rot=<0.0572 -0.0012 0.9982 0.0155>}
 calf_r {parent=thigh_r, loc=<41.918 0.2246 -0.0926>, rot=<0.0006 -0.0161 -0.0391 -0.9991>}
 foot_r {parent=calf_r, loc=<38.7057 -1.2438 -0.2452>, rot=<0.0975 -0.0474 0.0317 -0.9936>}
 ball_r {parent=foot_r, loc=<6.1614 -13.1469 -0.6807>, rot=<0.0001 -0.0001 -0.7063 -0.7079>}
 dfrm_leg_spline_09_r_ctrl_null {parent=foot_r, type=Null, loc=<0.0416 0.0328 0.1606>, rot=<0.0487 0.1016 -0.9936 0.0088>}
 dfrm_leg_spline_09_r_ctrl {
  controlType=EULER_TRANSFORM
  parent=dfrm_leg_spline_09_r_ctrl_null
  type=Control
  loc=<-19.7693 -3.2699 10.2921>
  rot=<0.0138 0.6747 -0.0126 -0.7379>
 }
 calf_twist_02_r {parent=calf_r, loc=<15.9268 0.0967 -0.0397>, rot=<0.0237 0.0095 -0.0035 -0.9997>}
 calf_twist_01_r {parent=calf_r, loc=<32.9335 0.0967 -0.0397>, rot=<0.0237 0.0095 -0.0035 -0.9997>}
 deform_knee_in_base_r {parent=calf_r, loc=<-1.2982 8.0356 0.0496>, rot=<0 0 0 1>}
 deform_knee_in_r {parent=deform_knee_in_base_r, loc=<0 0 0>, rot=<0 0 0 1>}
 deform_knee_out_base_r {parent=calf_r, loc=<-0.0002 -6.8739 0>, rot=<0 0 0 -1>}
 deform_knee_out_r {parent=deform_knee_out_base_r, loc=<0 0 0>, rot=<0 0 0 1>}
 deform_groin_base_r {parent=thigh_r, loc=<2.1908 -8.5661 1.7442>, rot=<0 0 0 -1>}
 deform_groin_r {parent=deform_groin_base_r, loc=<0 0 0>, rot=<0 0 0 -1>}
 thigh_twist_01_r {parent=thigh_r, loc=<21.6303 0 0>, rot=<0.0013 0 0 -1>}
 deform_glute_base_l {parent=pelvis, loc=<-10.1877 -11.6217 -12.899>, rot=<0.0275 0.0501 -0.4472 -0.8926>}
 deform_glute_l {parent=deform_glute_base_l, loc=<0 0 0>, rot=<0 0 0 1>}
 deform_glute_base_r {parent=pelvis, loc=<-10.1877 -11.6218 12.899>, rot=<0.0501 -0.0275 0.8926 -0.4472>}
 deform_glute_r {parent=deform_glute_base_r, loc=<0 0 0>, rot=<0 0 0 1>}
 dfrm_leg_spline_MSTR_l {parent=pelvis, loc=<-19.7392 0.5421 -14.153>, rot=<-0.0445 0 0.999 -0.0006>}
 dfrm_leg_spline_01_l {parent=dfrm_leg_spline_MSTR_l, loc=<0 0 0>, rot=<0 0 0 1>}
 dfrm_leg_spline_02_l {parent=dfrm_leg_spline_01_l, loc=<7.8326 0.0046 -0.003>, rot=<0 0 0 1>}
 dfrm_leg_spline_03_l {parent=dfrm_leg_spline_02_l, loc=<7.8346 0.0072 -0.0047>, rot=<0 0 0 1>}
 dfrm_leg_spline_04_l {parent=dfrm_leg_spline_03_l, loc=<7.836 0.009 -0.0059>, rot=<0 0 0 1>}
 dfrm_leg_spline_05_l {parent=dfrm_leg_spline_04_l, loc=<7.8376 0.0111 -0.0072>, rot=<0 0 0 1>}
 dfrm_leg_spline_06_l {parent=dfrm_leg_spline_05_l, loc=<7.8379 0.0115 -0.0075>, rot=<0 0 0 1>}
 dfrm_leg_spline_07_l {parent=dfrm_leg_spline_06_l, loc=<7.8362 0.0094 -0.0061>, rot=<0 0 0 1>}
 dfrm_leg_spline_08_l {parent=dfrm_leg_spline_07_l, loc=<7.8342 0.0067 -0.0043>, rot=<0 0 0 1>}
 dfrm_leg_spline_09_l {parent=dfrm_leg_spline_08_l, loc=<7.8319 0.0037 -0.0024>, rot=<0 0 0 1>}
 dfrm_thigh_l {parent=pelvis, loc=<-19.7392 0.5421 -14.0478>, rot=<0.0012 0.0572 0.0155 -0.9982>}
 dfrm_calf_l {parent=dfrm_thigh_l, loc=<-32.0794 -0.9241 0.3141>, rot=<0.0006 -0.0161 -0.0391 -0.9991>}
 dfrm_foot_l {parent=dfrm_calf_l, loc=<-30.5993 1.1349 0.3725>, rot=<0.0975 -0.0474 0.0317 -0.9936>}
 dfrm_knee_out_base_l {parent=dfrm_calf_l, loc=<-1.4166 7.7309 0.8374>, rot=<0 0 0 -1>}
 dfrm_knee_out_l {parent=dfrm_knee_out_base_l, loc=<0 0 0>, rot=<0 0 0 1>}
 dfrm_knee_in_base_l {parent=dfrm_calf_l, loc=<-0.1185 -7.1787 0.7878>, rot=<0 0 0 1>}
 dfrm_knee_in_l {parent=dfrm_knee_in_base_l, loc=<0 0 0>, rot=<0 0 0 1>}
 dfrm_thigh_r {parent=pelvis, loc=<-19.7392 0.5421 14.0478>, rot=<0.0572 -0.0012 0.9982 0.0155>}
 dfrm_calf_r {parent=dfrm_thigh_r, loc=<32.0324 0.8595 -0.3542>, rot=<0.0006 -0.0161 -0.0391 -0.9991>}
 dfrm_foot_r {parent=dfrm_calf_r, loc=<30.632 -1.1599 -0.2049>, rot=<0.0975 -0.0474 0.0317 -0.9936>}
 dfrm_knee_out_base_r {parent=dfrm_calf_r, loc=<1.4672 -7.6702 -0.7957>, rot=<0 0 0 -1>}
 dfrm_knee_out_r {parent=dfrm_knee_out_base_r, loc=<0 0 0>, rot=<0 0 0 1>}
 dfrm_knee_in_base_r {parent=dfrm_calf_r, loc=<0.1692 7.2394 -0.746>, rot=<0 0 0 -1>}
 dfrm_knee_in_r {parent=dfrm_knee_in_base_r, loc=<0 0 0>, rot=<0 0 0 1>}
 dfrm_leg_spline_MSTR_r {parent=pelvis, loc=<-19.7392 0.5421 14.153>, rot=<0 -0.0445 -0.0006 -0.999>}
 dfrm_leg_spline_01_r {parent=dfrm_leg_spline_MSTR_r, loc=<0 0 0>, rot=<0 -0.0002 -0.0003 -1>}
 dfrm_leg_spline_02_r {parent=dfrm_leg_spline_01_r, loc=<-7.8326 0 0>, rot=<0 -0.0001 -0.0002 -1>}
 dfrm_leg_spline_03_r {parent=dfrm_leg_spline_02_r, loc=<-7.8346 0 0>, rot=<0 -0.0001 -0.0001 -1>}
 dfrm_leg_spline_04_r {parent=dfrm_leg_spline_03_r, loc=<-7.836 0 0>, rot=<0 -0.0001 -0.0001 -1>}
 dfrm_leg_spline_05_r {parent=dfrm_leg_spline_04_r, loc=<-7.8376 0 0.0001>, rot=<0 0 0 -1>}
 dfrm_leg_spline_06_r {parent=dfrm_leg_spline_05_r, loc=<-7.8379 0 0>, rot=<0 0.0001 0.0001 -1>}
 dfrm_leg_spline_07_r {parent=dfrm_leg_spline_06_r, loc=<-7.8362 0 0>, rot=<0 0.0001 0.0002 -1>}
 dfrm_leg_spline_08_r {parent=dfrm_leg_spline_07_r, loc=<-7.8342 0 -0.0001>, rot=<0 0.0001 0.0002 -1>}
 dfrm_leg_spline_09_r {parent=dfrm_leg_spline_08_r, loc=<-7.8319 0 0.0001>, rot=<0 0 0 -1>}
 dfrm_leg_spline_01_l_ctrl_null {parent=pelvis, type=Null, loc=<-19.7392 0.5421 -14.153>, rot=<0.0447 0 -0.999 0.0009>}
 dfrm_leg_spline_01_l_ctrl {
  controlType=EULER_TRANSFORM
  parent=dfrm_leg_spline_01_l_ctrl_null
  type=Control
  loc=<71.0364 -3.6281 20.5809>
  rot=<0.7379 -0.0127 -0.6747 -0.0139>
 }
 dfrm_leg_spline_03_l_ctrl_null {parent=pelvis, type=Null, loc=<-35.3436 0.512 -15.5542>, rot=<0.0449 0.0001 -0.999 0.0012>}
 dfrm_leg_spline_03_l_ctrl {
  controlType=EULER_TRANSFORM
  parent=dfrm_leg_spline_03_l_ctrl_null
  type=Control
  loc=<55.3596 -3.6628 20.6027>
  rot=<0.738 -0.0129 -0.6745 -0.0141>
 }
 dfrm_leg_spline_05_l_ctrl_null {parent=pelvis, type=Null, loc=<-50.9538 0.4736 -16.9614>, rot=<0.045 0.0001 -0.999 0.0013>}
 dfrm_leg_spline_05_l_ctrl {
  controlType=EULER_TRANSFORM
  parent=dfrm_leg_spline_05_l_ctrl_null
  type=Control
  loc=<39.6806 -3.6779 20.6121>
  rot=<0.7381 -0.013 -0.6744 -0.0142>
 }
 dfrm_leg_spline_07_l_ctrl_null {parent=pelvis, type=Null, loc=<-66.5645 0.4345 -18.3691>, rot=<0.0448 0 -0.999 0.001>}
 dfrm_leg_spline_07_l_ctrl {
  controlType=EULER_TRANSFORM
  parent=dfrm_leg_spline_07_l_ctrl_null
  type=Control
  loc=<24.017 -3.6601 20.6012>
  rot=<0.7379 -0.0128 -0.6746 -0.014>
 }
 dfrm_leg_spline_01_r_ctrl_null {parent=pelvis, type=Null, loc=<-19.7392 0.5421 14.153>, rot=<0 0.0447 0.0009 0.999>}
 dfrm_leg_spline_01_r_ctrl {
  controlType=EULER_TRANSFORM
  parent=dfrm_leg_spline_01_r_ctrl_null
  type=Control
  loc=<-14.153 -0.8975 72.6759>
  rot=<0.0139 0.6747 -0.0127 -0.7379>
 }
 dfrm_leg_spline_03_r_ctrl_null {parent=pelvis, type=Null, loc=<-35.3436 0.512 15.5542>, rot=<-0.0001 0.0449 0.0012 0.999>}
 dfrm_leg_spline_03_r_ctrl {
  controlType=EULER_TRANSFORM
  parent=dfrm_leg_spline_03_r_ctrl_null
  type=Control
  loc=<-15.5542 -1.4865 57.0826>
  rot=<0.0141 0.6745 -0.0129 -0.738>
 }
 dfrm_leg_spline_05_r_ctrl_null {parent=pelvis, type=Null, loc=<-50.9538 0.4736 16.9614>, rot=<-0.0001 0.045 0.0013 0.999>}
 dfrm_leg_spline_05_r_ctrl {
  controlType=EULER_TRANSFORM
  parent=dfrm_leg_spline_05_r_ctrl_null
  type=Control
  loc=<-16.9614 -2.0841 41.4838>
  rot=<0.0142 0.6744 -0.013 -0.7381>
 }
 dfrm_leg_spline_07_r_ctrl_null {parent=pelvis, type=Null, loc=<-66.5645 0.4345 18.3691>, rot=<0 0.0448 0.001 0.999>}
 dfrm_leg_spline_07_r_ctrl {
  controlType=EULER_TRANSFORM
  parent=dfrm_leg_spline_07_r_ctrl_null
  type=Control
  loc=<-18.3691 -2.6824 25.8845>
  rot=<0.014 0.6746 -0.0128 -0.7379>
 }
 dfrm_thigh_up_extra_r {parent=pelvis, loc=<-19.739 0.542 14.0478>, rot=<0.0572 -0.0012 0.9982 0.0155>}
 dfrm_thigh_up_extra_l {parent=pelvis, loc=<-19.739 0.542 -14.0478>, rot=<0.0012 0.0572 0.0155 -0.9982>}
 pelvis_extra {parent=pelvis, loc=<-16.4375 0.4237 0>, rot=<0.0012 0.0572 0.0155 -0.9982>}
 attach {parent=root, loc=<0 0 0>, rot=<0 0 0 1>}
 ik_hand_root {parent=root, loc=<0 0 0>, rot=<0 0 0 1>}
 ik_hand_gun {parent=ik_hand_root, loc=<-64.462 4.8638 101.1961>, rot=<0.6084 0.1793 0.4613 -0.6205>}
 ik_hand_r {parent=ik_hand_gun, loc=<0 0 0>, rot=<0 0 0 1>}
 ik_hand_l {parent=ik_hand_gun, loc=<65.7708 101.9258 43.6659>, rot=<-0.5102 -0.7906 -0.3387 0>}
 ik_foot_root {parent=root, loc=<0 0 0>, rot=<0 0 0 1>}
 ik_foot_l {parent=ik_foot_root, loc=<19.6189 -3.2051 10.3339>, rot=<0.0753 0.7018 0.0677 -0.7051>}
 ik_foot_r {parent=ik_foot_root, loc=<-19.6189 -3.2051 10.3339>, rot=<0.7051 0.0677 -0.7018 0.0753>}
 root_fp {parent=root, loc=<0 0 0>, rot=<0 0 0 1>}
 camera_root {parent=root_fp, loc=<0 0 0>, rot=<0 0 0 1>}
 camera {parent=camera_root, loc=<0 0 0>, rot=<0 0 0 1>}
 attach_fp {parent=root_fp, loc=<0 0 0>, rot=<0 0 0 1>}
 attach_emote_01 {parent=root, loc=<0 0 0>, rot=<0 0 0 1>}
 attach_emote_02 {parent=root, loc=<0 0 0>, rot=<0 0 0 1>}
 attach_emote_03 {parent=root, loc=<0 0 0>, rot=<0 0 0 1>}
 warp_point_01 {parent=root, loc=<0 0 0>, rot=<0 0 0 1>}
 warp_point_02 {parent=root, loc=<0 0 0>, rot=<0 0 0 1>}
 warp_point_03 {parent=root, loc=<0 0 0>, rot=<0 0 0 1>}
 thigh_l_ctrl {controlType=EULER_TRANSFORM, type=Control, loc=<0 0 0>, rot=<0 0 0 1>}
 calf_l_ctrl {controlType=EULER_TRANSFORM, parent=thigh_l_ctrl, type=Control, loc=<0 0 0>, rot=<0 0 0 1>}
 foot_l_ctrl {controlType=EULER_TRANSFORM, parent=calf_l_ctrl, type=Control, loc=<0 0 0>, rot=<0 0 0 1>}
 thigh_r_ctrl {controlType=EULER_TRANSFORM, type=Control, loc=<0 0 0>, rot=<0 0 0 1>}
 calf_r_ctrl {controlType=EULER_TRANSFORM, parent=thigh_r_ctrl, type=Control, loc=<0 0 0>, rot=<0 0 0 1>}
 foot_r_ctrl {controlType=EULER_TRANSFORM, parent=calf_r_ctrl, type=Control, loc=<0 0 0>, rot=<0 0 0 1>}
 pelvis_ctrl {controlType=EULER_TRANSFORM, type=Control, loc=<0 0 0>, rot=<0 0 0 1>}
]
variables [
 Crouching {type=bool, default=False}
 Emote {type=double, default=0}
 FrontEndValue {type=double, default=0}
 Idle FE {type=double, default=0}
 L_legSpline_ctrls {
  type="TArray<FRigElementKey>"
  default=((Type=Control,Name="dfrm_leg_spline_01_l_ctrl"),(Type=Control,Name="dfrm_leg_spline_03_l_ctrl"),(Type=Control,Name="dfrm_leg_spline_05_l_ctrl"),(Type=Control,Name="dfrm_leg_spline_07_l_ctrl"),(Type=Control,Name="dfrm_leg_spline_09_l_ctrl"))
 }
 PistolTargeting {type=double, default=0}
 R_legSpline_ctrls {
  type="TArray<FRigElementKey>"
  default=((Type=Control,Name="dfrm_leg_spline_01_r_ctrl"),(Type=Control,Name="dfrm_leg_spline_03_r_ctrl"),(Type=Control,Name="dfrm_leg_spline_05_r_ctrl"),(Type=Control,Name="dfrm_leg_spline_07_r_ctrl"),(Type=Control,Name="dfrm_leg_spline_09_r_ctrl"))
 }
 kneeSliding {type=bool, default=False}
]
