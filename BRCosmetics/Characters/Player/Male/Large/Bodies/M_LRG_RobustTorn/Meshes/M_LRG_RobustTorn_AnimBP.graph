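# Post-process style AnimBP for the M_LRG_RobustTorn male body cosmetic, copying the attached parent's pose and correcting it.
# A Control Rig takes crouch, knee-slide, pistol and front-end inputs, then additive offsets reshape forearm twist bones.
# Pectoral bones get stacked Anim Dynamics passes, angular then linear springs, for secondary jiggle.
# EventGraph derives the rig floats, plus crouch-spring and weapon-tag forearm-hide values the AnimGraph never reads.
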AnimBlueprint /BRCosmetics/Characters/Player/Male/Large/Bodies/M_LRG_RobustTorn/Meshes/M_LRG_RobustTorn_AnimBP
target_skeleton=&/M_LRG_RobustTorn_Skeleton
graphs [
 AnimGraph {nodes [
   n1 = Control Rig!(
    Source=Copy Pose From Mesh() {bUseAttachedParent=true}
    Alpha
    Crouching:bool=Get bIsCrouching()
    Idle FE:double=Get Idle FE() -> (Idle FE:double)
    kneeSliding:bool=Get bIsKneeSliding()
    PistolTargeting:double=Get TargetingPistol() -> (TargetingPistol:double)
    FrontEndValue:double=Get FrontEndValue() -> (FrontEndValue:double)
   ) -> (Pose) {controlRigAssetReference {blueprintRigClass=&/M_LRG_RobustTorn_CtrlRig.M_LRG_RobustTorn_CtrlRig_C}}
   n2 = "Transform (Modify) Bone - Bone: deform_lowerarm_twist_02_r"!(
    ComponentPose=Local To Component(LocalPose=@n1)
    bAlphaBoolEnabled:bool=True
    Translation="-0.000,5.298,-1.783"
    Rotation="0,0,0"
    Scale="0.316,0.284,0.316"
   ) -> (Pose) {
    alphaInputType=Bool
    rotationMode=BMM_Additive
    rotationSpace=BCS_BoneSpace
    scaleMode=BMM_Additive
    scaleSpace=BCS_BoneSpace
    translationMode=BMM_Additive
    translationSpace=BCS_BoneSpace
    boneToModify {boneName=deform_lowerarm_twist_02_r}
    scale=<0.316 0.284 0.316>
    translation=<0 5.298 -1.7828>
   }
   n3 = "Transform (Modify) Bone - Bone: lowerarm_twist_02_r"!(ComponentPose=@n2, bAlphaBoolEnabled:bool=True, Translation="0.574,4.793,-3.051", Rotation="0,0,0", Scale="1.000,0.455,0.400") -> (Pose) {
    alphaInputType=Bool
    rotationMode=BMM_Additive
    rotationSpace=BCS_BoneSpace
    scaleMode=BMM_Additive
    scaleSpace=BCS_BoneSpace
    translationMode=BMM_Additive
    translationSpace=BCS_BoneSpace
    boneToModify {boneName=lowerarm_twist_02_r}
    scale {y=0.455, z=0.4}
    translation=<0.5739 4.793 -3.0511>
   }
   n4 = "Transform (Modify) Bone - Bone: deform_lowerarm_twist_02_l"!(ComponentPose=@n3, bAlphaBoolEnabled:bool=True, Translation="0.000,-1.673,2.015", Rotation="0.000,0.000,-4.527", Scale="0.316,0.284,0.316") -> (Pose) {
    alphaInputType=Bool
    rotationMode=BMM_Additive
    rotationSpace=BCS_BoneSpace
    scaleMode=BMM_Additive
    scaleSpace=BCS_BoneSpace
    translationMode=BMM_Additive
    translationSpace=BCS_BoneSpace
    boneToModify {boneName=deform_lowerarm_twist_02_l}
    rotation=<rot 0.0001 0 -4.527>
    scale=<0.316 0.284 0.316>
    translation=<0 -1.673 2.0147>
   }
   n5 = "Transform (Modify) Bone - Bone: lowerarm_twist_02_l"!(ComponentPose=@n4, bAlphaBoolEnabled:bool=True, Translation="-0.000,-1.999,2.688", Rotation="0,0,0", Scale="1.000,0.605,0.368") -> (Pose) {
    alphaInputType=Bool
    rotationMode=BMM_Additive
    rotationSpace=BCS_BoneSpace
    scaleMode=BMM_Additive
    scaleSpace=BCS_BoneSpace
    translationMode=BMM_Additive
    translationSpace=BCS_BoneSpace
    boneToModify {boneName=lowerarm_twist_02_l}
    scale {y=0.605, z=0.368}
    translation {y=-1.999, z=2.688}
   }
   n6 = "Transform (Modify) Bone - Bone: dyn_L_pectoralDynamics"(ComponentPose=@n5, Alpha=1, Translation="0,0,1", Rotation="0,0,0", Scale="1,1,1") {translationMode=BMM_Additive, translationSpace=BCS_BoneSpace, boneToModify {boneName=dyn_L_pectoralDynamics}, translation {z=1}}
   n7 = "Transform (Modify) Bone - Bone: dyn_R_pectoralDynamics"(ComponentPose=@n6, Alpha=1, Translation="0,0,1", Rotation="0,0,0", Scale="1,1,1") {translationMode=BMM_Additive, translationSpace=BCS_BoneSpace, boneToModify {boneName=dyn_R_pectoralDynamics}, translation {z=1}}
   n8 = Anim Dynamics - Bone: dyn_L_pectoralDynamics(ComponentPose=@n7, Alpha=1, ExternalForce="0,0,0") {
    angularBiasOverride=1
    angularDampingOverride=0.9999
    angularSpringConstant=4000
    bAngularSpring=true
    bGravityOverrideInSimSpace=true
    bOverrideAngularBias=true
    bOverrideAngularDamping=true
    bUseGravityOverride=true
    simulationSpace=RootRelative
    boundBone {boneName=dyn_L_pectoralDynamics}
    componentAppliedLinearAccClamp=<500 500 500>
    gravityOverride {z=-980}
    physicsBodyDefinitions [{
      boundBone {boneName=dyn_L_pectoralDynamics}
      boxExtents=<20 20 20>
      constraintSetup {angularLimitsMax=<5 0 5>, angularLimitsMin=<-5 0 -5>, angularTarget=<1 0 0>}
      localJointOffset=<2 0 0>
     }]
    relativeSpaceBone {boneName=spine_04}
    simSpaceSettings {maxAngularAcceleration=35, maxAngularVelocity=7, simSpaceAngularAlpha=0.7}
   }
   n9 = Anim Dynamics - Bone: dyn_R_pectoralDynamics(ComponentPose=@n8, Alpha=1, ExternalForce="0,0,0") {
    angularBiasOverride=1
    angularDampingOverride=0.9999
    angularSpringConstant=4000
    bAngularSpring=true
    bGravityOverrideInSimSpace=true
    bOverrideAngularBias=true
    bOverrideAngularDamping=true
    bUseGravityOverride=true
    simulationSpace=RootRelative
    boundBone {boneName=dyn_R_pectoralDynamics}
    componentAppliedLinearAccClamp=<500 500 500>
    gravityOverride {z=-980}
    physicsBodyDefinitions [{
      boundBone {boneName=dyn_R_pectoralDynamics}
      boxExtents=<21 21 21>
      constraintSetup {angularLimitsMax=<5 0 5>, angularLimitsMin=<-5 0 -5>, angularTarget=<1 0 0>}
      localJointOffset=<-2 0 0>
     }]
    relativeSpaceBone {boneName=spine_04}
    simSpaceSettings {maxAngularAcceleration=35, maxAngularVelocity=7, simSpaceAngularAlpha=0.7}
   }
   n10 = Anim Dynamics - Bone: dyn_L_pectoralDynamics(ComponentPose=@n9, Alpha=1, ExternalForce="0,0,0") {
    angularBiasOverride=1
    angularDampingOverride=0.99
    angularSpringConstant=10000
    bGravityOverrideInSimSpace=true
    bLinearSpring=true
    bOverrideAngularBias=true
    bOverrideLinearDamping=true
    bUseGravityOverride=true
    linearDampingOverride=1
    linearSpringConstant=200
    simulationSpace=RootRelative
    boundBone {boneName=dyn_L_pectoralDynamics}
    componentAppliedLinearAccClamp=<500 500 500>
    gravityOverride {z=-450}
    physicsBodyDefinitions [{
      boundBone {boneName=dyn_L_pectoralDynamics}
      boxExtents=<90 90 90>
      constraintSetup {angularLimitsMax=<0 0 0>, angularLimitsMin=<0 0 0>, angularTarget=<1 0 0>, linearAxesMax=<0 0.5 2>, linearAxesMin=<0 -0.5 -1>}
      localJointOffset=<0.2 0 0>
     }]
    relativeSpaceBone {boneName=spine_04}
    simSpaceSettings {maxAngularAcceleration=35, maxAngularVelocity=7, simSpaceAngularAlpha=0.7}
   }
   n11 = Anim Dynamics - Bone: dyn_R_pectoralDynamics(ComponentPose=@n10, Alpha=1, ExternalForce="0,0,0") {
    angularBiasOverride=1
    angularDampingOverride=0.99
    angularSpringConstant=10000
    bGravityOverrideInSimSpace=true
    bLinearSpring=true
    bOverrideAngularBias=true
    bOverrideLinearDamping=true
    bUseGravityOverride=true
    linearDampingOverride=1
    linearSpringConstant=200
    simulationSpace=RootRelative
    boundBone {boneName=dyn_R_pectoralDynamics}
    componentAppliedLinearAccClamp=<500 500 500>
    gravityOverride {z=-450}
    physicsBodyDefinitions [{
      boundBone {boneName=dyn_R_pectoralDynamics}
      boxExtents=<80 80 80>
      constraintSetup {angularLimitsMax=<0 0 0>, angularLimitsMin=<0 0 0>, angularTarget=<1 0 0>, linearAxesMax=<0 0.5 1>, linearAxesMin=<0 -0.5 -2>}
      localJointOffset=<-0.2 0 0>
     }]
    relativeSpaceBone {boneName=spine_04}
    simSpaceSettings {maxAngularAcceleration=35, maxAngularVelocity=7, simSpaceAngularAlpha=0.7}
   }
   Output Pose(Result=Component To Local(ComponentPose=@n11)) {layerGroup=None}
  ]}
 EventGraph {nodes [
   Event BlueprintUpdateAnimation()
   > n1 = Sequence()
   n2 = NOT Boolean(A=Get bIsSurfaceSwimming())
   n3 = NOT Boolean(A=Get bIsMoving2D())
   n4 = NOT Boolean(A=Get bIsPlayingEmote())
   n5 = NOT Boolean(A=Get bIsSkyDiving())
   n6 = Get bIsInFrontEnd()
   > Set Idle FE(Idle FE:double="To Float (Boolean)"(InBool=AND Boolean!(A=@n3, B=@n4, C:bool=@n5, D:bool=@n2, E:bool=@n6) -> (ReturnValue))) -> (Output_Get:double)
   > n7 = Branch(Condition=Get bIsCrouching())
   > Set SPRCrouchingValue(SPRCrouchingValue:double=-100) -> (Output_Get:double)
   Set SPRCrouchingValue(execute=@n7.else, SPRCrouchingValue=-150)
   n8 = NOT Boolean(A=Get bIsSkyDiving())
   n9 = NOT Boolean(A=Get bIsTacticalSprinting())
   n10 = NOT Boolean(A=Get bIsPlayingEmote())
   n11 = "Equal (Enum)"(A:EFortWeaponCoreAnimation Enum=Get WeaponCoreAnim!() -> (WeaponCoreAnim), B:EFortWeaponCoreAnimation Enum=Pistol)
   Set TargetingPistol(execute=@n1.then_1, TargetingPistol:double="To Float (Boolean)"(InBool=AND Boolean!(A=@n11, B=@n10, C=@n8, D=@n9) -> (ReturnValue))) -> (Output_Get:double)
   > n12 = Branch(Condition=NOT Boolean(A=Get bIsInFrontEnd()))
   > Set FrontEndValue(FrontEndValue:double=0) -> (Output_Get:double)
   n13 = Get bIsSkyDiving()
   n14 = Get bIsPlayingEmote()
   n15 = Branch(execute=@n12.else, Condition=AND Boolean!(A=Get bIsInFrontEnd(), B=NOT Boolean(A=@n13), C=NOT Boolean(A=@n14)) -> (ReturnValue))
   > Set FrontEndValue(FrontEndValue=1)
   Set FrontEndValue(execute=@n15.else, FrontEndValue=0.5)
   Event OnWeaponChanged()
   > n16 = Cast To FortPawn!(execute, Object=GetOwningActor()) -> (then, CastFailed, AsFort Pawn:Fort Pawn Object Reference) {is=title K2Node_DynamicCast}
   n17 = Get CurrentWeapon!(self:Fort Pawn Object Reference=@n16) -> (CurrentWeapon:Fort Weapon Object Reference)
   > Set LowerForeamRightHidden(LowerForeamRightHidden:double="To Float (Boolean)"(InBool=WeaponDataHasAnyMatchingGameplayTags!(
      self:Fort Weapon Object Reference=@n17
      TagContainer:Gameplay Tag Container Structure=(GameplayTags=((TagName="CharacterPart.BoneSet.LowerForearmRight")))
     ) -> (ReturnValue:bool))) -> (Output_Get:double)
  ]}
]
variables [
 Detected_Pistol {type=number}
 FrontEndValue {type=number}
 Head {type=SkeletalMeshComponentBudgeted}
 Head_Mesh {type=SkeletalMeshComponent}
 Idle FE {type=number}
 LowerForeamRightHidden {type=number}
 SPRCrouchingValue {type=number, default=-150}
 TargetingPistol {type=number}
]
